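# Unlit translucent cloud material for the CrispRover skydiving contrail, usable on Niagara mesh particles and spline meshes.
# Fake lighting dots a panning, flattened cloud normal against world up to blend time-of-day shadow and light colors, tinted by particle color.
# Opacity combines separate above/below view-angle fades, UV-V tip ramp, near-camera fade, particle alpha, vertex red and depth fade (skipped on mobile).
# Dynamic parameters drive tiling and pan speed, squaring PanningSpeedX and ignoring PanningSpeedY.
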Material /BRCosmetics/Athena/Effects/Particles/Skydiving/NS_Contrail_CrispRover/FX/M_CrispRover_Cloud
graph {
 properties {
  bUsedWithNiagaraMeshParticles=true
  bUsedWithSplineMeshes=true
  blendMode=BLEND_Translucent
  pixelDepthOffsetMode=PDOM_Legacy
  shadingModel=MSM_Unlit
  usedShadingModels=Unlit
 }
 nodes [
  n1 = ParticleColor()
  n2 = ScalarParameter() {defaultValue=1.5, parameterName=Opacity}
  n3 = ComponentMask(None=CameraVectorWS()) {b=true}
  n4 = ScalarParameter() {defaultValue=0.01, parameterName=offset}
  n5 = Clamp(None=Divide(A=Add(A=@n3, B=@n4), B=ScalarParameter() {defaultValue=0.01, parameterName=Div}))
  n6 = LinearInterpolate(A=Multiply(A=@n2, B=ScalarParameter() {defaultValue=1, parameterName=OpacityUnder}) {constB=4}, B=@n2, Alpha=@n5)
  n7 = Multiply(
   A=@n6
   B=MaterialFunctionCall("FadeSensitivity (optional)"=ScalarParameter() {defaultValue=3.4, parameterName=AngleFadeSensitivityU}) {materialFunction=&FadeBasedOnViewAngle}.Result
  )
  n8 = Multiply(
   A=@n6
   B=MaterialFunctionCall("FadeSensitivity (optional)"=ScalarParameter() {defaultValue=2.6, parameterName=AngleFadeSensitivity}) {materialFunction=&FadeBasedOnViewAngle}.Result
  )
  n9 = ComponentMask(None=TextureCoordinate()) {g=true}
  n10 = Multiply(A=Multiply(A=@n9, B=OneMinus(None=@n9)), B=ScalarParameter() {defaultValue=1.5, parameterName=ScaleCloudTipRamp})
  n11 = Multiply(
   A=Multiply(A=LinearInterpolate(A=@n7, B=@n8, Alpha=@n5), B=Clamp(None=@n10))
   B=MaterialFunctionCall(Distance=ScalarParameter() {defaultValue=100, parameterName=CameraDistanceFade}) {materialFunction=&Near_Camera_Fade}.Result
  )
  n12 = DepthFade(FadeDistance=ScalarParameter() {defaultValue=1000, parameterName=Opacity_Depth_FadeDistance})
  n13 = Constant() {r=1}
  Multiply(
   A=Multiply(A=Multiply(A=@n1.A, B=@n11), B=ShadingPathSwitch(Default=@n12, Deferred=@n12, Mobile=@n13) {desc=Migrated from FeatureLevelSwitch})
   B=VertexColor().R
  )
  n14 = MaterialFunctionCall(
   Color_Morning=VectorParameter() {parameterName=ShadowColorDay, defaultValue=<rgba 0.1024 0.1595 0.2049 1>}.RGB
   Color_Day=VectorParameter() {parameterName=ShadowColorMorning, defaultValue=<rgba 0.0955 0.174 0.191 1>}.RGB
   Color_Evening=VectorParameter() {parameterName=ShadowColorEvening, defaultValue=<rgba 0.1181 0.0708 0.0827 1>}.RGB
   Color_Night=VectorParameter() {parameterName=ShadowColorNight, defaultValue=<rgba 0.0924 0.1253 0.1319 1>}.RGB
  ) {materialFunction=&Color_Time_of_Day}
  n15 = MaterialFunctionCall(
   Color_Morning=VectorParameter() {parameterName=LightColorMorning, defaultValue=<rgba 0.1301 0.0648 0.1301 1>}.RGB
   Color_Day=VectorParameter() {parameterName=LightColorDay, defaultValue=<rgba 0.251 0.2902 0.3059 1>}.RGB
   Color_Evening=VectorParameter() {parameterName=LightColorEvening, defaultValue=<rgba 0.3467 0.1046 0.0296 1>}.RGB
   Color_Night=VectorParameter() {parameterName=LightColorNight, defaultValue=<rgba 0.0976 0.1356 0.15 1>}.RGB
  ) {materialFunction=&Color_Time_of_Day}
  n16 = Transform(None=Normalize(VectorInput=Constant3Vector() {constant=<rgba 0 0 1 0>})) {transformSourceType=TRANSFORMSOURCE_World, transformType=TRANSFORM_Tangent}
  n17 = DynamicParameter() {paramNames [PanningSpeedX, PanningSpeedY, TileX, TileY]}
  n18 = ScalarParameter() {defaultValue=2, parameterName=Tile_N_X}
  n19 = TextureCoordinate()
  n20 = ScalarParameter() {defaultValue=2, parameterName=Tile_N_Y}
  n21 = DynamicParameter() {paramNames [PanningSpeedX, PanningSpeedY, TileX, TileY]}
  n22 = AppendVector(
   A=Multiply(A=Multiply(A=@n17.TileX, B=@n18), B=ComponentMask(None=@n19) {r=true})
   B=Multiply(A=ComponentMask(None=@n19) {g=true}, B=Multiply(A=@n20, B=@n21.TileY))
  )
  n23 = DynamicParameter() {paramNames [PanningSpeedX, PanningSpeedY, TileX, TileY]}
  n24 = Multiply(
   A=ComponentMask(None=VectorParameter() {parameterName=Speed, defaultValue=<rgba 0.02 0.02 0 0>}.RGB) {g=true, r=true}
   B=Multiply(A=@n23.PanningSpeedX, B=@n23.PanningSpeedX)
  )
  n25 = MaterialFunctionCall(
   Normal=TextureSample(UVs=Panner(Coordinate=@n22, Speed=@n24)) {samplerType=SAMPLERTYPE_LinearColor, texture=/Game/VisualThreatManager/StormVisuals/LargeClouds/T_StormCloudMobile}.RGB
   Flatness=ScalarParameter() {defaultValue=0.15, parameterName=FlattenDetailNormal}
  ) {materialFunction=&FlattenNormal}
  n26 = LinearInterpolate(
   A=@n25.Result
   B=LinearInterpolate(A=@n25.Result, B=VertexNormalWS())
   Alpha=ScalarParameter() {defaultValue=0.1, parameterName=LerpVertNormal}
  )
  n27 = Power(
   Base=ConstantBiasScale(None=DotProduct(A=@n16, B=@n26)) {bias=0.5, scale=0.6}
   Exp=ScalarParameter() {defaultValue=2, parameterName=SoftenShadow}
  )
  n28 = Multiply(A=LinearInterpolate(A=Multiply(A=@n14.Result, B=@n15.Result), B=@n15.Result, Alpha=Clamp(None=@n27) {minDefault=0.2})) {constB=3}
  Reroute(None=MaterialFunctionCall(Emissive=Multiply(A=@n28, B=@n1.RGB)) {materialFunction=&MF_FX_EmissiveAdjustments}.Result)
 ]
}
