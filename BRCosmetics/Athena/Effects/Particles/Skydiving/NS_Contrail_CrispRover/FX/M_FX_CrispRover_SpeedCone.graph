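# Translucent two-sided Niagara material for the speed cone in the CrispRover skydiving contrail.
# Color blends side to center along UV V, boosted through time-of-day HDR and emissive functions, with a pure particle-color switch.
# Opacity combines two panning noise textures, center and top gradients, an optional center hole, Fresnel and optional depth fade.
# A sine wobble along local Z, driven by dynamic parameters, feeds a reroute named WPO.
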Material /BRCosmetics/Athena/Effects/Particles/Skydiving/NS_Contrail_CrispRover/FX/M_FX_CrispRover_SpeedCone
graph {
 properties {
  bUsedWithNiagaraMeshParticles=true
  bUsedWithNiagaraRibbons=true
  bUsedWithNiagaraSprites=true
  blendMode=BLEND_Translucent
  pixelDepthOffsetMode=PDOM_Legacy
  twoSided=true
  usedShadingModels=Default Lit
 }
 nodes [
  n1 = DynamicParameter() {parameterIndex=3, defaultValue=<rgba 1 1 0.25 1>, paramNames [Wing_FlapSpeed, Param2, Param3, MainSpeed]}
  n2 = Multiply(A=Time(), B=@n1.MainSpeed)
  Add(A=@n2, B=Multiply(A=Constant() {r=-0.1}))
  n3 = ParticleColor()
  n4 = ComponentMask(None=TextureCoordinate()) {g=true}
  n5 = ScalarParameter() {defaultValue=6, parameterName=Central Contast}
  n6 = Saturate(None=Multiply(A=ScalarParameter() {defaultValue=2, parameterName=Central_Alpha Intensity}, B=Power(Base=@n4, Exp=@n5) {constExponent=10}))
  n7 = TextureCoordinate()
  n8 = ComponentMask(None=VectorParameter() {parameterName=NoiseB_Tile, defaultValue=<rgba 2 4 0 0>}.RGB) {g=true, r=true}
  n9 = ComponentMask(None=VectorParameter() {parameterName=PanSpeedXY, defaultValue=<rgba 0 1 0 0>}.RGB) {g=true, r=true}
  n10 = Constant2Vector() {g=1, r=-1}
  n11 = TextureSampleParameter2D(UVs=Panner(Coordinate=Multiply(A=@n7, B=@n8), Speed=Multiply(A=@n9, B=@n10)) {speedX=-1, speedY=2.5}) {parameterName=NoiseB, texture=/Game/Global/Textures/Default/PointSplatter/T_Tiling_Dust}
  n12 = Power(Base=Add(A=@n6, B=@n11.G), Exp=ScalarParameter() {defaultValue=2, parameterName=EmissiveExp})
  n13 = Multiply(A=@n3.RGB, B=Add(A=@n12, B=@n6))
  n14 = LinearInterpolate(
   A=VectorParameter() {parameterName=SideColor, defaultValue=<rgba 0.0334 0.548 1.5 0>}.RGB
   B=VectorParameter() {parameterName=CenterColor, defaultValue=<rgba 0.4531 0.7178 1 0>}.RGB
   Alpha=@n6
  )
  n15 = Multiply(A=Saturate(None=@n14), B=@n3.RGB)
  n16 = Multiply(
   A=@n14
   B=MaterialFunctionCall(
    Color_Morning=ScalarParameter() {defaultValue=10, parameterName=HDRMult_3}
    Color_Day=ScalarParameter() {defaultValue=10, parameterName=HDRMult_2}
    Color_Evening=ScalarParameter() {defaultValue=5, parameterName=HDRMult_1}
    Color_Night=ScalarParameter() {defaultValue=2, parameterName=HDRMult}
   ) {materialFunction=&Color_Time_of_Day}.Result
  )
  n17 = MaterialFunctionCall(
   LDR Input=StaticSwitchParameter(True=Clamp(None=@n15), False=@n15) {parameterName=UseClampOnHDR}
   HDR Tint=StaticSwitchParameter(True=Multiply(A=@n3.RGB, B=@n16), False=@n16) {parameterName=UseClampOnHDR}
   HDR Intensity Multiplier=LinearInterpolate(
    A=ScalarParameter() {defaultValue=2, parameterName=HDRIntensitySides}
    B=ScalarParameter() {defaultValue=2, parameterName=HDRIntensity}
    Alpha=@n6
   )
  ) {materialFunction=&DeriveHDRfromLDR}
  n18 = ScalarParameter() {parameterName=Emis}
  n19 = Multiply(A=Multiply(A=@n12, B=@n17.RemappedRange), B=@n18)
  MaterialFunctionCall(Emissive=MaterialFunctionCall(Emissive=StaticSwitchParameter(True=@n13, False=@n19) {parameterName=UsePureParticleColor}) {materialFunction=&MF_FX_EmissiveAdjustments}.Result) {materialFunction=&MF_FX_EmissiveAdjustments}
  n20 = Normalize(VectorInput=Transform(None=Constant3Vector() {constant=<rgba 0 0 1 0>}) {transformSourceType=TRANSFORMSOURCE_Local})
  n21 = DynamicParameter() {parameterIndex=1, defaultValue=<rgba 1.25 6 2 0.25>, paramNames [TinyWobbleSpeed, UpDownMove, DiagnolBackForthMove, Param4]}
  n22 = DynamicParameter() {defaultValue=<rgba 1 1 0.1 0.025>, paramNames [MoveBodyOrWing, BatAnimFreq, WingBend, WingMiddleBend]}
  n23 = Sine(None=Multiply(A=@n2, B=Multiply(A=@n21.TinyWobbleSpeed, B=@n22.BatAnimFreq)))
  n24 = DynamicParameter() {parameterIndex=1, defaultValue=<rgba 1.25 6 2 0.25>, paramNames [TinyWobbleSpeed, UpDownMove, DiagnolBackForthMove, Param4]}
  NamedRerouteDeclaration(None=Multiply(A=Multiply(A=@n20, B=@n23), B=@n24.UpDownMove)) {name=WPO, nodeColor=<rgba 0 0.2751 1 1>}
  NamedRerouteUsage()
  NamedRerouteUsage()
  NamedRerouteUsage()
  n25 = ComponentMask(None=TextureCoordinate()) {g=true}
  n26 = TextureCoordinate()
  n27 = ComponentMask(None=VectorParameter() {parameterName=NoiseA_Tile, defaultValue=<rgba 3 1 0 0>}.RGB) {g=true, r=true}
  n28 = TextureSampleParameter2D(UVs=Panner(Coordinate=Multiply(A=@n26, B=@n27), Speed=@n9) {speedX=1, speedY=1}) {parameterName=NoiseA, texture=/Game/Weapons/WeaponSkins/Textures/T_Swirly_Noise_highContrast}
  n29 = ScalarParameter() {defaultValue=1, parameterName=Streaks_LerpA}
  n30 = ScalarParameter() {defaultValue=3, parameterName=Streaks_LerpB}
  n31 = OneMinus(None=ComponentMask(None=TextureCoordinate()) {g=true})
  n32 = Saturate(None=Power(Base=Add(A=@n11.G, B=@n28.G), Exp=LinearInterpolate(A=@n29, B=@n30, Alpha=@n31) {constA=0.6, constB=4}))
  n33 = Multiply(A=Power(Base=OneMinus(None=@n25), Exp=ScalarParameter() {defaultValue=6, parameterName=Central Contast}) {constExponent=10}, B=@n32)
  n34 = Add(A=@n6, B=@n32)
  n35 = ComponentMask(None=TextureCoordinate()) {g=true}
  n36 = ScalarParameter() {defaultValue=3, parameterName=TopFade}
  n37 = Multiply(
   A=Saturate(None=StaticSwitchParameter(True=@n33, False=@n34) {parameterName=CenterHole})
   B=Saturate(None=Power(Base=@n35, Exp=@n36) {constExponent=10})
  )
  n38 = CameraVectorWS()
  n39 = Multiply(A=TwoSidedSign(), B=VertexNormalWS())
  n40 = Saturate(None=Power(Base=DotProduct(A=@n38, B=@n39), Exp=ScalarParameter() {defaultValue=4, parameterName=Fresnel Exp_1}))
  n41 = Multiply(
   A=LinearInterpolate(A=ScalarParameter() {defaultValue=2, parameterName=OpacSides}, B=ScalarParameter() {defaultValue=2, parameterName=OpacCenter}, Alpha=@n4)
   B=Multiply(A=@n3.A, B=Multiply(A=@n37, B=@n40))
  )
  n42 = DepthFade(FadeDistance=ScalarParameter() {parameterName=FadeDistance})
  n43 = Constant() {r=1}
  StaticSwitchParameter(True=Multiply(A=@n41, B=ShadingPathSwitch(Default=@n42, Deferred=@n42, Mobile=@n43) {desc=Migrated from FeatureLevelSwitch}), False=@n41) {parameterName=Use_DepthFade}
  n44 = OneMinus(None=@n18)
  StaticSwitchParameter(True=Multiply(A=Clamp(None=@n44), B=@n13), False=@n17.RemappedRange) {parameterName=UsePureParticleColor}
 ]
}
