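# Niagara skydiving contrail cosmetic CrispRover: a bird mesh accompanies the player amid drifting cloud puffs.
# Bird tracks a skeletal mesh bone with waveform wander and timeline-driven catch-up along skydive velocity, banking and turning by ParaGlideLeanAlpha.
# CloudLarge and CloudMed burst sprites in timed spawn groups, gated by system spawning, that partly inherit owner movement.
# Front-end modifiers change spawning and motion between gameplay and menus; Cloud and Cloud_FE mesh emitters are disabled.
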NiagaraSystem /BRCosmetics/Athena/Effects/Particles/Skydiving/NS_Contrail_CrispRover/FX/NS_Contrail_CrispRover
dependencies {
 usedDataInterfaces [
  &NiagaraDataInterfaceCurve
  &NiagaraDataInterfaceColorCurve
  &NiagaraDataInterfaceVectorCurve
  &NiagaraDataInterfaceCamera
  &NiagaraDataInterfaceArrayFloat
 ]
 usedDynamicInputs [
  &MakeCustomLinearColorFromBool
  &Boolean_OrOperation
  &Divide_Float
 ]
 usedModules [
  &self:SystemSpawnScript.SystemScriptSource.SystemScriptGraph.NiagaraNodeAssignment_3.SetVariables_#1
  /Game/Effects/Niagara/Modules/Skydiving/FN_DivingAttributes/v5/FN_DivingAttributes
  &SystemState
  &self:SystemSpawnScript.SystemScriptSource.SystemScriptGraph.NiagaraNodeAssignment_0.SetVariables_#2
  &Timeline
  &self:SystemSpawnScript.SystemScriptSource.SystemScriptGraph.NiagaraNodeAssignment_1.SetVariables_#3
  &self:SystemSpawnScript.SystemScriptSource.SystemScriptGraph.NiagaraNodeAssignment_2.SetVariables_#4
 ]
}
emitters {
 Bird {
  emitterUpdateScript [
   &EmitterState()
   &SpawnBurst_Instantaneous(
    Age:NiagaraFloat=$Emitter.Age
    Spawn Count=&MakeIntFromFloat(Float=&FN_FrontEndFloatModifier(Gameplay=1, Front End=0))
    Use Spawn Probability=false
    UseLoopCountLimit=false
   )
   &InterpolateOverTime(Rate Of Change=50, Target Vector Value:Vector3f=$Engine.Owner.Velocity, Variable Type=::NewEnumerator2)
  ]
  particleSpawnScript [
   &InitializeParticle(
    Color Mode=::NewEnumerator0
    Lifetime=5
    Mesh Scale Mode=::NewEnumerator3
    Position Offset=<0 0 1000>
    UsePositionOffset=true
    Write Material Random=true
   )
   &DynamicMaterialParameters(
    Index 0 Param 1=0.15
    Index 0 Param 2=&Multiply_Float(A=$Particles.Lifetime, B=&RandomRangeFloat(Minimum=0.16, Maximum=0.18))
    Index 0 Param 3=0.1
    Index 0 Param 4=0.05
    Module.Param0WriteEnabled=true
    Module.Param1WriteEnabled=true
    Module.Param2WriteEnabled=true
    Module.Param3WriteEnabled=true
   ) {module=DynamicMaterialParameters001}
   &InitialMeshOrientation(Mesh Orientation Mode=::NewEnumerator7, Use Rotation Vector=true)
  ]
  particleUpdateScript [
   &ParticleState(Kill Particles When Lifetime Has Elapsed=false, Let Infinitely Lived Particles Die When Emitter Deactivates=true)
   /SetVariables(Particles.Ambient:Vector3f=&Multiply_VectorByFloat(
     Vector=&MakeVector(
      X=&Waveform(
       "Waveform [1]"=::NewEnumerator12
       "[1] Input Position"=&RandomRangeFloat(Minimum=0, Maximum=1000)
       "[1] Phase Scale"=0.5
       "[1] Frequency"=1
       "[1] Amplitude Min/Max"=<-200 200>
      )
      Y=&Waveform(
       "Waveform [1]"=::NewEnumerator12
       "[1] Input Position"=&RandomRangeFloat(Minimum=1200, Maximum=1800)
       "[1] Phase Scale"=0.5
       "[1] Frequency"=1
       "[1] Amplitude Min/Max"=<0 200>
      )
      Z=&Waveform(
       "Waveform [1]"=::NewEnumerator12
       "[1] Input Position"=&RandomRangeFloat(Minimum=2200, Maximum=2800)
       "[1] Frequency"=1
       "[1] Amplitude Min/Max"=<-150 50>
      )
     )
     Float=0.35
    )) {isSetParameters=true, module=SetVariables_#5}
   &InterpolateOverTime(
    Rate Of Change=3
    Target Float Value:NiagaraFloat=&MakeCustomFloatFromBool(Bool=$System.Timeline001.IsRewinding, True Float=0, False Float=1)
   ) {module=InterpolateOverTime004}
   /SetVariables(Particles.CatchupVector:Vector3f=&Multiply_VectorByFloat(
     Vector=&Multiply_VectorByFloat(
      Vector=&NormalizeVector(Vector To Normalize=$System.SkydiveVelocity, Safe Normalize=true, Safe Fallback Threshold=0, Safe Fallback Vector=<0 0 0.001>)
      Float=-75
     )
     Float=&Lerp_Float(
      A=&FloatFromCurve(
       FloatCurve=DataInterface() {propertyValues {Curve {keys [
           {interpMode=RCIM_Cubic, tangentMode=RCTM_User, time=0, value=0}
           {interpMode=RCIM_Cubic, tangentMode=RCTM_User, time=0.35, value=-0.35}
           {tangentMode=RCTM_User, time=1, value=0}
          ]}}}
       CurveIndex=$System.Timeline001.NormalizedTime
      )
      B=&FloatFromCurve(
       FloatCurve=DataInterface() {propertyValues {Curve {keys [
           {arriveTangent=3.3827, arriveTangentWeight=1.1106, interpMode=RCIM_Cubic, leaveTangentWeight=1.1106, time=0, value=0}
           {arriveTangentWeight=0.5241, interpMode=RCIM_Cubic, leaveTangentWeight=0.1667, time=0.5, value=0.65}
           <at 1 0 RCIM_Cubic>
          ]}}}
       CurveIndex=$System.Timeline001.NormalizedTime
      )
      Alpha=$Particles.InterpolateOverTime004.InterpolateOverTime_Float.Moving Average
     )
    )) {isSetParameters=true}
   &SkeletalMeshLocation(
    Bone / Socket Index=5
    Bone Sampling Mode=::NewEnumerator4
    MaskSpawnGroup=false
    Orientation Sampling=::NewEnumerator1
    Sampled Position Offset=&Add_Vector(
     A=$Particles.CatchupVector
     B=/Niagara/DynamicInputs/Vectors/RotateVector(
      Vector To Rotate:Vector3f=&Add_Vector(A=$Particles.Ambient, B=<0 0 200>)
      Yaw:NiagaraFloat=&Waveform(
       "Waveform [1]"=::NewEnumerator12
       "[1] Input Position"=&RandomRangeFloat(Minimum=200000, Maximum=100000)
       "[1] Frequency"=1.25
       "[1] Amplitude Min/Max"=<7 -7>
      )
      Pitch:NiagaraFloat=0
      Roll:NiagaraFloat=0
      Rotation Rate:NiagaraFloat=1
     )
    )
    Skeletal Mesh=$User.SkeletalMesh
    Use Sampled Mesh Attribute as Ribbon ID=false
   )
   &ScaleMeshSize(Scale Factor=&VectorFromFloat(Value=&FloatFromCurve(
      FloatCurve=DataInterface() {propertyValues {Curve {keys [{arriveTangent=1.0186, interpMode=RCIM_Cubic, leaveTangent=1.0186, tangentMode=RCTM_User, time=0, value=0}, <at 1 1 RCIM_Cubic>]}}}
      CurveIndex=$Particles.Age
     )))
   &Drag(Drag=2, Use Linear Drag=true, Use Rotational Drag=false)
   &SolveForcesAndVelocity(Clamp Velocity=false, Limit Acceleration=false)
   &UpdateMeshOrientation(
    Bank Rate:NiagaraFloat=&Lerp_Float(
     A=&Lerp_Float(
      A=1
      B=30
      Alpha=&Waveform(
       "Waveform [1]"=::NewEnumerator12
       "[1] Input Position"=&RandomRangeFloat(Minimum=0, Maximum=1000)
       "[1] Phase Scale"=50000
       "[1] Frequency"=0.5
       "[1] Amplitude Min/Max"=<0 1>
      )
     )
     B=900
     Alpha=$User.ParaGlideLeanAlpha
    )
    Banking Angle Min / Max:Vector2f=&Lerp_Vector2(A=<-10 10>, B=<-1 1>, Alpha=$User.ParaGlideLeanAlpha)
    "Foward Vector (Mesh Space)":Vector3f=&Lerp_Vector(A=<0.5 0 0.5>, B=<1 0 0>, Alpha=$User.ParaGlideLeanAlpha)
    Orientation Method=::NewEnumerator3
    Particle Velocity:Vector3f=$Emitter.InterpolateOverTime.InterpolateOverTime_V3_.Moving Average
    Rotation Decay Rate In Degrees:NiagaraFloat=&Lerp_Float(A=150, B=500, Alpha=$User.ParaGlideLeanAlpha)
    Up Vector:Vector3f=$Engine.Owner.SystemYAxis
    Upright Roll Correction Speed:NiagaraFloat=&Lerp_Float(A=25, B=50, Alpha=$User.ParaGlideLeanAlpha)
    Velocity Orientation Convergence Rate:NiagaraFloat=&Lerp_Float(
     A=&Lerp_Float(
      A=3
      B=12
      Alpha=&Waveform(
       "Waveform [1]"=::NewEnumerator12
       "[1] Input Position"=&RandomRangeFloat(Minimum=0, Maximum=100)
       "[1] Frequency"=5
       "[1] Amplitude Min/Max"=<0 1>
      )
     )
     B=25
     Alpha=$User.ParaGlideLeanAlpha
    )
   ) {module=UpdateMeshOrientation001}
   &InterpolateOverTime(Rate Of Change=5, Target Quat Value:Quat4f=$Particles.MeshOrientation, Variable Type=::NewEnumerator5) {module=InterpolateOverTime002}
   /SetVariables(Particles.New Orient:Quat4f=$Particles.InterpolateOverTime002.InterpolateOverTime_Quaternion.Moving Average) {isSetParameters=true, module=SetVariables_#6}
   /SetVariables(Particles.MeshOrientation=$Particles.New Orient) {isSetParameters=true, module=SetVariables_#7}
   &self:SphereBlockingVolume_05(Sphere Origin:Vector3f=/Niagara/DynamicInputs/Helpers/SimulationPosition(), Sphere Radius=80) {module=SphereBlockingVolume_05}
   &DynamicMaterialParameters(
    Index 0 Param 1=&RemapRange(Input Value To Remap=$System.Timeline002.NormalizedTime, Input Min=0, Input Max=1, Output Min=0.15, Output Max=1)
    Module.Param0WriteEnabled=true
    Module.Param1WriteEnabled=false
    Module.Param2WriteEnabled=false
    Module.Param3WriteEnabled=false
   )
   /SetVariables(Particles.Position=&AddVectorToPosition(
     Position=$Particles.Position
     Vector=&MakeVector(
      X=0
      Y=0
      Z=&FloatFromCurve(
       FloatCurve=DataInterface() {propertyValues {Curve {keys [{arriveTangent=-177.2428, interpMode=RCIM_Cubic, leaveTangent=-177.2428, tangentMode=RCTM_User, time=0, value=150}, <at 1 0 RCIM_Cubic>]}}}
       CurveIndex=$Particles.Age
      )
     )
    )) {isSetParameters=true}
   /SetVariables(Particles.SpeedConeColor:LinearColor=&MakeLinearColorFromVectorAndFloat("Vector (RGB)"=<1 1 1>, "Float (Alpha)"=$User.ParaGlideLeanAlpha)) {isSetParameters=true}
  ]
  rendererClasses [&NiagaraMeshRendererProperties, &NiagaraSpriteRendererProperties]
  renderers [
   {rendererClass=&NiagaraMeshRendererProperties, rendererIndex=0}
   {rendererClass=&NiagaraMeshRendererProperties, rendererIndex=1}
   {rendererClass=&NiagaraSpriteRendererProperties, rendererIndex=2}
  ]
 }
 Cloud {
  bEnabled=false
  emitterUpdateScript [
   &EmitterState(
    Enable Distance Culling=true
    Life Cycle Mode=::NewEnumerator1
    Loop Duration=2
    Max Distance Response=::SleepAndClearParticles
    Module.bApplySpawnCountScaleByDistance=false
    Module.bUseMaxDistance=true
    Module.bUseMinDistance=false
    Reset Age On Awaken=true
    Scalability Mode=::NewEnumerator1
    Scale Spawn Count=true
    Spawn Count Scale By Distance=DataInterface() {propertyValues {Curve {keys [<at 0 0>,<at 0.3 0.95>,<at 0.75 1>,<at 1 0>]}}}
    UseLoopDelay=false
   )
   &SpawnRate(
    Spawn Probability=&FN_FrontEndFloatModifier(Gameplay=1, Front End=0)
    SpawnRate=&Lerp_Float(A=2.5, B=5, Alpha=$User.ParaGlideLeanAlpha)
    Use Spawn Probability=true
   )
  ]
  particleSpawnScript [
   /SetVariables(Particles.MeshIndex=&RandomRangeInt(Minimum=0, Maximum=2)) {isSetParameters=true, module=SetVariables_#8}
   &InitializeParticle(
    Color=&ColorFromCurve(
     ColorCurve=DataInterface() {propertyValues {
       AlphaCurve {keys [<at 0 1 RCIM_Cubic>,<at 1 1 RCIM_Cubic>]}
       BlueCurve {keys [<at 0 1 RCIM_Cubic>,<at 0.7008 1>,<at 1 1 RCIM_Cubic>]}
       GreenCurve {keys [<at 0 1 RCIM_Cubic>,<at 0.7008 0.8409>,<at 1 0.9 RCIM_Cubic>]}
       RedCurve {keys [<at 0 1 RCIM_Cubic>,<at 0.7008 0.8>,<at 1 0.9473 RCIM_Cubic>]}
      }}
     CurveIndex=&RandomRangeFloat(Minimum=0, Maximum=1)
    )
    Lifetime Max=2.5
    Lifetime Min=2
    Lifetime Mode=::NewEnumerator1
    Mesh Scale=&Multiply_Vector(
     A=&RandomRangeVector(Minimum=<0.6 0.8 0.8>, Maximum=<0.8 1 1>)
     B=&VectorFromFloat(Value=&RandomRangeFloat(Minimum=1, Maximum=1.5))
    )
    Mesh Scale Mode=::NewEnumerator4
    UsePositionOffset=false
    Write Material Random=true
   ) {module=InitializeParticle001}
   &InitialMeshOrientation(
    Mesh Orientation Mode=::NewEnumerator7
    Rotation=&SelectVectorFromArray(
     Array Sampling Mode=::NewEnumerator0
     Direct Array Index=?
     Vector Selection Array=DataInterface() {propertyValues {FloatData [<0.05 0 0.5>, <0 0 0>]}}
     Random Seed=0
    )
    Use Rotation Vector=true
   )
   &ShapeLocation(
    Disc Coverage=-1
    MaskSpawnGroup=false
    Offset=&Lerp_Vector(A=<0 0 0>, B=<0 100 -250>, Alpha=$User.ParaGlideLeanAlpha)
    Offset Coordinate Space=::Local
    Offset Mode=::NewEnumerator0
    Override Seed=false
    Ring Radius=&RandomRangeFloat(Minimum=80, Maximum=100)
    Shape Primitive=::NewEnumerator4
    UseOwnerScale=true
   )
  ]
  particleUpdateScript [
   &ParticleState()
   &ScaleMeshSize(Scale Factor=&MakeVector(
     X=&Lerp_Float(A=1, B=1.1, Alpha=$User.ParaGlideLeanAlpha)
     Y=1
     Z=&Lerp_Float(A=1, B=1.1, Alpha=$User.ParaGlideLeanAlpha)
    )) {module=ScaleMeshSize001}
   &ScaleMeshSize(Scale Factor=/Niagara/DynamicInputs/Scalability/ScalabilityDistanceBasedVector(
     Curve:NiagaraDataInterfaceVectorCurve=DataInterface() {propertyValues {
       XCurve {keys [{tangentMode=RCTM_User, time=0, value=1}, {tangentMode=RCTM_User, time=1, value=0}]}
       YCurve {keys [{tangentMode=RCTM_User, time=0, value=1}, {tangentMode=RCTM_User, time=1, value=0}]}
       ZCurve {keys [{tangentMode=RCTM_User, time=0, value=1}, {tangentMode=RCTM_User, time=1, value=0}]}
      }}
     CurveIndex:NiagaraFloat=$Emitter.Scalability.DistanceFraction
    ))
   /Niagara/Modules/Update/Orientation/V2/OrientMeshToVector(
    Look At Direction:Vector3f=&ReturnCameraVector(Camera Vectors=::NewEnumerator3, Coordinate Space=::Simulation)
    Override Quat=false
    UseOptionalSideAxis=false
    UseOptionalUpAxis=false
   )
   &ScaleColor(
    Scale Alpha=&FloatFromCurve(
     FloatCurve=DataInterface() {propertyValues {Curve {keys [
         {interpMode=RCIM_Cubic, tangentMode=RCTM_User, time=0.1, value=0}
         {interpMode=RCIM_Cubic, tangentMode=RCTM_User, time=0.5, value=1}
         <at 0.9 0>
        ]}}}
     Scale Curve=&FloatFromCurve(
      FloatCurve=DataInterface() {propertyValues {Curve {keys [
          {arriveTangent=-0.0101, interpMode=RCIM_Cubic, leaveTangent=-0.0101, tangentMode=RCTM_User, time=0, value=0.75}
          {arriveTangent=3.0014, interpMode=RCIM_Cubic, leaveTangent=3.0014, tangentMode=RCTM_User, time=0.1813, value=0.9875}
          {arriveTangent=1, time=0.2, value=1}
          <at 1 1>
         ]}}}
      CurveIndex=$Emitter.Scalability.DistanceFraction
     )
    )
    ScaleA=true
    ScaleRGB=true
   )
   &DynamicMaterialParameters(
    Index 0 Param 1=&RandomRangeFloat(Minimum=2, Maximum=3)
    Index 0 Param 2=&RandomRangeFloat(Minimum=2, Maximum=3)
    Index 0 Param 3=1
    Index 0 Param 4=1
    Module.Param0WriteEnabled=true
    Module.Param1WriteEnabled=true
    Module.Param2WriteEnabled=true
    Module.Param3WriteEnabled=true
   )
   &InheritSourceMovement(
    Applied Position Scale=&VectorFromFloat(Value=&FloatFromCurve(FloatCurve=DataInterface() {propertyValues {Curve {keys [
          {arriveTangent=-0.3402, interpMode=RCIM_Cubic, leaveTangent=-0.3402, tangentMode=RCTM_User, time=0, value=1}
          {arriveTangent=-1.6874, interpMode=RCIM_Cubic, leaveTangent=-1.6874, tangentMode=RCTM_User, time=1, value=0.2}
         ]}}}))
    ApplyForce=false
    ApplyPosition=true
    ApplyPositionFalloff=false
   )
   &SolveForcesAndVelocity(Clamp Velocity=false, Limit Acceleration=false)
  ]
  rendererClasses [&NiagaraMeshRendererProperties]
  renderers [
   {rendererClass=&NiagaraMeshRendererProperties, rendererIndex=0}
   {rendererClass=&NiagaraMeshRendererProperties, rendererIndex=1}
   {rendererClass=&NiagaraMeshRendererProperties, rendererIndex=2}
   {rendererClass=&NiagaraMeshRendererProperties, rendererIndex=3}
  ]
 }
 CloudLarge {
  emitterUpdateScript [
   &EmitterState(Inactive Response=::NewEnumerator1, Life Cycle Mode=::NewEnumerator1, UseLoopDelay=false)
   &SpawnBurst_Instantaneous(
    Spawn Count=&MakeCustomIntFromBool(Bool=$System.Spawning, True INT=1, False INT=0)
    Use Spawn Probability=false
    UseLoopCountLimit=false
   )
   &SpawnBurst_Instantaneous(
    Spawn Count=&MakeCustomIntFromBool(Bool=$System.Spawning, True INT=1, False INT=0)
    Spawn Group=1
    Spawn Time=2
    Use Spawn Probability=false
    UseLoopCountLimit=false
   ) {module=SpawnBurst_Instantaneous001}
   &SpawnBurst_Instantaneous(
    Spawn Count=&MakeCustomIntFromBool(Bool=$System.Spawning, True INT=1, False INT=0)
    Spawn Group=2
    Spawn Time=3
    Use Spawn Probability=false
    UseLoopCountLimit=false
   ) {module=SpawnBurst_Instantaneous002}
   &SpawnBurst_Instantaneous(
    Spawn Count=&MakeCustomIntFromBool(Bool=$System.Spawning, True INT=1, False INT=0)
    Spawn Group=3
    Spawn Time=5
    Use Spawn Probability=false
    UseLoopCountLimit=false
   ) {module=SpawnBurst_Instantaneous003}
   &SpawnBurst_Instantaneous(
    Spawn Count=&MakeCustomIntFromBool(Bool=$System.Spawning, True INT=1, False INT=0)
    Spawn Group=4
    Spawn Time=7
    Use Spawn Probability=false
    UseLoopCountLimit=false
   ) {module=SpawnBurst_Instantaneous004}
   &SpawnBurst_Instantaneous(
    Spawn Count=&MakeCustomIntFromBool(Bool=$System.Spawning, True INT=1, False INT=0)
    Spawn Group=5
    Spawn Time=8.5
    Use Spawn Probability=false
    UseLoopCountLimit=false
   ) {module=SpawnBurst_Instantaneous005}
   &SpawnRate(SpawnRate=3, Use Spawn Probability=false) {enabled=false}
  ]
  particleSpawnScript [
   &InitializeParticle(
    Color=$System.CloudColor
    Lifetime=5
    Sprite Rotation Angle=&RandomRangeFloat(Minimum=-45, Maximum=45)
    Sprite Rotation Mode=::NewEnumerator0
    Sprite Size Mode=::NewEnumerator3
    Uniform Sprite Size=&Multiply_Float(
     A=90
     B=&SelectFloatFromArray(Direct Array Index=$Emitter.SpawnGroup, Float Selection Array=DataInterface() {propertyValues {FloatData [1,1.3,1.5,1,1.3,1.5]}})
    )
    UsePositionOffset=false
    Write Material Random=true
   ) {module=InitializeParticle001}
   /Niagara/Modules/Spawn/Initialization/InitializeParticle(
    Color=&UniformRangedLinearColor(Minimum=<rgba 1.5 1.3724 1.4638 1>, Maximum=<rgba 0.2913 0.3405 0.526 1>)
    Lifetime=&UniformRangedFloat(Minimum=10, Maximum=12)
    Module.Write Color=true
    Module.Write Lifetime=true
    Module.Write Mass=true
    Module.Write Material Random=true
    Module.Write Position=true
    Module.Write Scale=false
    Module.Write SpriteRotation=true
    Module.Write SpriteSize=true
    Sprite Rotation=&NormalizedAngleToDegrees(NormalizedAngle=&UniformRangedFloat(Minimum=0, Maximum=360))
    Sprite Size=&Vector2DFromFloat(Value=&RandomRangeFloat(Minimum=35, Maximum=45))
   ) {enabled=false}
   /SetVariables(Particles.Position=&MaskVectorBySpawnGroup(
     Masked Vector=&Add_Vector(A=Linked() {linkedVariable {name=Engine.Owner.Position, type=NiagaraPosition}}, B=<0 0 10>)
     Passthrough Vector=Linked() {linkedVariable {name=Particles.Position, type=NiagaraPosition}}
     Spawn Group Mask=1
    )) {enabled=false, isSetParameters=true, module=SetVariables_#9}
   &CameraOffset(Camera Offset Amount=&RandomRangeFloat(Minimum=5, Maximum=10)) {module=CameraOffset001}
   &ShapeLocation(
    MaskSpawnGroup=false
    Offset=&FN_FrontEndVector3Modifier(Gameplay=<0 0 -300>, Front End=<0 0 -125>)
    Offset Mode=::NewEnumerator0
    Override Seed=false
    Ring / Disc Distribution Mode:ENiagaraSphereDistributionMode=::NewEnumerator1
    Ring Radius=200
    Shape Primitive=::NewEnumerator4
    U Position=&SelectFloatFromArray(Direct Array Index=$Emitter.SpawnGroup, Float Selection Array=DataInterface() {propertyValues {FloatData [0.1,0.5,0.3,0.8,0.2,0.9]}})
    UseOwnerScale=true
   ) {module=ShapeLocation001}
   &ShapeLocation(
    Cylinder Height=1
    Cylinder Radius=&SelectFloatFromArray(Direct Array Index=$Emitter.SpawnGroup, Float Selection Array=DataInterface() {propertyValues {FloatData [15,22,25,15,20,22]}})
    MaskSpawnGroup=false
    Non Uniform Scale=<1.5 0.7 0.7>
    Override Seed=false
    Shape Primitive=::NewEnumerator1
    Surface Only=true
    UseOwnerScale=true
   )
   &AddVelocity(Velocity=<0 -35 100>, Velocity Speed Scale=&FN_FrontEndFloatModifier(Gameplay=0, Front End=1))
   &CylinderLocation(
    Cylinder Height=1
    Cylinder Radius=30
    MaskSpawnGroup=false
    Module.Enable Lathe Profile=false
    Module.Override Randomness=false
    Module.Surface Only=true
    Surface Only Band Thickness=1
    Use Endcaps In Surface Only Mode=false
   ) {enabled=false}
  ]
  particleUpdateScript [
   &ParticleState()
   &SpriteRotationRate(Rotation Rate=&Multiply_Float(
     A=&RandomRangeFloat(Minimum=5, Maximum=10)
     B=&SelectFloatFromArray(
      Array Sampling Mode=::NewEnumerator0
      Direct Array Index=?
      Float Selection Array=DataInterface() {propertyValues {FloatData [-1,1]}}
      Random Seed=0
     )
    ))
   &GravityForce(
    Coordinate Space=::Local
    Gravity=&MakeVector(
     X=0
     Y=&FN_FrontEndFloatModifier(
      Gameplay=0
      Front End=&RemapRange(Input Value To Remap=$System.Timeline002.NormalizedTime, Input Min=0, Input Max=1, Output Min=0, Output Max=-50)
     )
     Z=&FN_FrontEndFloatModifier(
      Gameplay=0
      Front End=&RemapRange(Input Value To Remap=$System.Timeline002.NormalizedTime, Input Min=0, Input Max=1, Output Min=0, Output Max=150)
     )
    )
   ) {module=GravityForce001}
   &ScaleSpriteSize(Scale Factor=&Vector2DFromFloat(Value=&FloatFromCurve(FloatCurve=DataInterface() {propertyValues {Curve {keys [
          {arriveTangent=1.3248, interpMode=RCIM_Cubic, leaveTangent=1.3248, tangentMode=RCTM_User, time=0, value=0.7}
          {arriveTangent=0.4626, interpMode=RCIM_Cubic, leaveTangent=0.4626, tangentMode=RCTM_User, time=0.9978, value=1.202}
         ]}}}))) {enabled=false}
   &Drag(Drag=3, Use Linear Drag=true, Use Rotational Drag=true) {enabled=false}
   &InheritSourceMovement(
    Applied Position Scale=&VectorFromFloat(Value=&FloatFromCurve(FloatCurve=DataInterface() {propertyValues {Curve {keys [
          {arriveTangent=-0.3402, interpMode=RCIM_Cubic, leaveTangent=-0.3402, tangentMode=RCTM_User, time=0, value=1}
          {arriveTangent=-1.6874, interpMode=RCIM_Cubic, leaveTangent=-1.6874, tangentMode=RCTM_User, time=1, value=0.2}
         ]}}}))
    ApplyForce=false
    ApplyPosition=true
    ApplyPositionFalloff=false
   )
   &SolveForcesAndVelocity(Clamp Velocity=false, Limit Acceleration=false)
   &ScaleColor(
    Scale Alpha=&FloatFromCurve(FloatCurve=DataInterface() {propertyValues {Curve {keys [
         {arriveTangent=1.4796, interpMode=RCIM_Cubic, leaveTangent=1.4796, tangentMode=RCTM_User, time=0, value=0}
         {interpMode=RCIM_Cubic, tangentMode=RCTM_User, time=0.5, value=1}
         {arriveTangent=-2.9378, interpMode=RCIM_Cubic, leaveTangent=-2.9378, tangentMode=RCTM_User, time=1, value=0}
        ]}}})
    ScaleA=true
    ScaleRGB=true
   )
   &DynamicMaterialParameters(
    Index 0 Param 1=&Add_Float(
     A=$Particles.RandomFloat
     B=&FloatFromCurve(
      FloatCurve=DataInterface() {propertyValues {Curve {keys [
          {arriveTangent=1.0192, interpMode=RCIM_Cubic, leaveTangent=1.0192, tangentMode=RCTM_User, time=0, value=1}
          {arriveTangent=0.2834, interpMode=RCIM_Cubic, leaveTangent=0.2834, tangentMode=RCTM_User, time=1, value=1.5}
         ]}}}
      CurveIndex=&Add_Float(A=&Multiply_Float(A=$Particles.Initial.MaterialRandom, B=0.5), B=$Particles.NormalizedAge)
      Scale Curve=0.4
     )
    )
    Index 0 Param 3=&FloatFromCurve(FloatCurve=DataInterface() {propertyValues {Curve {keys [
         {arriveTangent=-53.9199, interpMode=RCIM_Cubic, leaveTangent=-8.9727, tangentMode=RCTM_User, time=0, value=1}
         {arriveTangent=0.2347, interpMode=RCIM_Cubic, leaveTangent=0.2723, tangentMode=RCTM_Break, time=0.25, value=0}
        ]}}})
    Index 0 Param 4=&FloatFromCurve(FloatCurve=DataInterface() {propertyValues {Curve {keys [<at 0.5 1>,<at 1 0>]}}})
    Index 2 Param 1=&FloatFromCurve(FloatCurve=DataInterface() {propertyValues {Curve {keys [{interpMode=RCIM_Cubic, leaveTangent=1.4301, tangentMode=RCTM_Break, time=0, value=0}, <at 1 1>]}}})
    Index 2 Param 2=&FloatFromCurve(
     FloatCurve=DataInterface() {propertyValues {Curve {keys [
         {arriveTangent=0.6777, interpMode=RCIM_Cubic, leaveTangent=0.6777, tangentMode=RCTM_User, time=0, value=0.1}
         {arriveTangent=0.0259, interpMode=RCIM_Cubic, leaveTangent=0.0259, tangentMode=RCTM_User, time=1, value=0.4}
        ]}}}
     Scale Curve=&RandomRangeFloat(Minimum=0.2, Maximum=0.5)
    )
    Index 2 Param 3=&FloatFromCurve(FloatCurve=DataInterface() {propertyValues {Curve {keys [
         {interpMode=RCIM_Cubic, leaveTangent=2.3064, tangentMode=RCTM_Break, time=0, value=1}
         {interpMode=RCIM_Cubic, leaveTangent=11.1233, tangentMode=RCTM_Break, time=1, value=2}
        ]}}})
    Index 2 Param 4=&FloatFromCurve(
     FloatCurve=DataInterface() {propertyValues {Curve {keys [
         {arriveTangent=-0.4764, interpMode=RCIM_Cubic, leaveTangent=-0.4764, tangentMode=RCTM_User, time=0, value=1}
         {interpMode=RCIM_Cubic, leaveTangent=-1.2956, tangentMode=RCTM_Break, time=1, value=0.75}
        ]}}}
     Scale Curve=1.6
    )
    Module.Param0WriteEnabled=true
    Module.Param0WriteEnabled2=true
    Module.Param1WriteEnabled=false
    Module.Param1WriteEnabled2=true
    Module.Param2WriteEnabled=true
    Module.Param2WriteEnabled2=true
    Module.Param3WriteEnabled=true
    Module.Param3WriteEnabled2=true
    Write Parameter Index 2=true
   )
   &CameraOffset(Camera Offset Amount=35)
   &KillParticles(Kill Particles=&SetBoolByFloatComparison(
     A=&MakeFloatFromVector(
      Vector=&ConvertVectorToPosition(
       Input Position=Linked() {linkedVariable {name=Particles.Position, type=NiagaraPosition}}
       Conversion Operation:ENiagara_LWCConvertVecToPos=::NewEnumerator1
      )
      Channel=::NewEnumerator2
     )
     B=8000
     Comparison Type=::NewEnumerator6
    )) {enabled=false}
  ]
  rendererClasses [&NiagaraSpriteRendererProperties]
  renderers [
   {rendererClass=&NiagaraSpriteRendererProperties}
   {rendererClass=&NiagaraSpriteRendererProperties, rendererIndex=1}
  ]
 }
 CloudMed {
  emitterUpdateScript [
   &EmitterState(Inactive Response=::NewEnumerator1, Life Cycle Mode=::NewEnumerator1, UseLoopDelay=false)
   &SpawnBurst_Instantaneous(
    Spawn Count=&MakeCustomIntFromBool(Bool=$System.Spawning, True INT=10, False INT=0)
    Use Spawn Probability=false
    UseLoopCountLimit=false
   )
   &SpawnBurst_Instantaneous(
    Spawn Count=&MakeCustomIntFromBool(Bool=$System.Spawning, True INT=15, False INT=0)
    Spawn Group=1
    Spawn Time=2
    Use Spawn Probability=false
    UseLoopCountLimit=false
   ) {module=SpawnBurst_Instantaneous001}
   &SpawnBurst_Instantaneous(
    Spawn Count=&MakeCustomIntFromBool(Bool=$System.Spawning, True INT=20, False INT=0)
    Spawn Group=2
    Spawn Time=3
    Use Spawn Probability=false
    UseLoopCountLimit=false
   ) {module=SpawnBurst_Instantaneous002}
   &SpawnBurst_Instantaneous(
    Spawn Count=&MakeCustomIntFromBool(Bool=$System.Spawning, True INT=10, False INT=0)
    Spawn Group=3
    Spawn Time=5
    Use Spawn Probability=false
    UseLoopCountLimit=false
   ) {module=SpawnBurst_Instantaneous003}
   &SpawnBurst_Instantaneous(
    Spawn Count=&MakeCustomIntFromBool(Bool=$System.Spawning, True INT=15, False INT=0)
    Spawn Group=4
    Spawn Time=7
    Use Spawn Probability=false
    UseLoopCountLimit=false
   ) {module=SpawnBurst_Instantaneous004}
   &SpawnBurst_Instantaneous(
    Spawn Count=&MakeCustomIntFromBool(Bool=$System.Spawning, True INT=20, False INT=0)
    Spawn Group=5
    Spawn Time=8.5
    Use Spawn Probability=false
    UseLoopCountLimit=false
   ) {module=SpawnBurst_Instantaneous005}
   &SpawnRate(SpawnRate=3, Use Spawn Probability=false) {enabled=false}
  ]
  particleSpawnScript [
   /Niagara/Modules/Spawn/Initialization/V2/InitializeParticle(
    Color=$System.CloudColor
    Lifetime=5
    Sprite Rotation Angle=&RandomRangeFloat(Minimum=-45, Maximum=45)
    Sprite Rotation Mode=::NewEnumerator0
    Sprite Size Mode=::NewEnumerator3
    Uniform Sprite Size=&Multiply_Float(
     A=&RandomRangeFloat(Minimum=30, Maximum=45)
     B=&SelectFloatFromArray(Direct Array Index=$Emitter.SpawnGroup, Float Selection Array=DataInterface() {propertyValues {FloatData [1,1.3,1.5,1,1.3,1.5]}})
    )
    UsePositionOffset=false
    Write Material Random=true
   ) {module=InitializeParticle001}
   /Niagara/Modules/Spawn/Initialization/InitializeParticle(
    Color=&UniformRangedLinearColor(Minimum=<rgba 1.5 1.3724 1.4638 1>, Maximum=<rgba 0.2913 0.3405 0.526 1>)
    Lifetime=&UniformRangedFloat(Minimum=10, Maximum=12)
    Module.Write Color=true
    Module.Write Lifetime=true
    Module.Write Mass=true
    Module.Write Material Random=true
    Module.Write Position=true
    Module.Write Scale=false
    Module.Write SpriteRotation=true
    Module.Write SpriteSize=true
    Sprite Rotation=&NormalizedAngleToDegrees(NormalizedAngle=&UniformRangedFloat(Minimum=0, Maximum=360))
    Sprite Size=&Vector2DFromFloat(Value=&RandomRangeFloat(Minimum=35, Maximum=45))
   ) {enabled=false}
   /SetVariables(Particles.Position=&MaskVectorBySpawnGroup(
     Masked Vector=&Add_Vector(A=Linked() {linkedVariable {name=Engine.Owner.Position, type=NiagaraPosition}}, B=<0 0 10>)
     Passthrough Vector=Linked() {linkedVariable {name=Particles.Position, type=NiagaraPosition}}
     Spawn Group Mask=1
    )) {enabled=false, isSetParameters=true, module=SetVariables_#9}
   &CameraOffset(Camera Offset Amount=&RandomRangeFloat(Minimum=5, Maximum=10)) {module=CameraOffset001}
   &ShapeLocation(
    MaskSpawnGroup=false
    Offset=&FN_FrontEndVector3Modifier(Gameplay=<0 0 -300>, Front End=<0 0 -125>)
    Offset Mode=::NewEnumerator0
    Override Seed=false
    Ring / Disc Distribution Mode:ENiagaraSphereDistributionMode=::NewEnumerator1
    Ring Radius=200
    Shape Primitive=::NewEnumerator4
    U Position=&SelectFloatFromArray(Direct Array Index=$Emitter.SpawnGroup, Float Selection Array=DataInterface() {propertyValues {FloatData [0.1,0.5,0.3,0.8,0.2,0.9]}})
    UseOwnerScale=true
   ) {module=ShapeLocation001}
   &ShapeLocation(
    Cylinder Height=15
    Cylinder Radius=&SelectFloatFromArray(Direct Array Index=$Emitter.SpawnGroup, Float Selection Array=DataInterface() {propertyValues {FloatData [30,45,50,30,40,45]}})
    MaskSpawnGroup=false
    Non Uniform Scale=<1.5 0.7 0.7>
    Override Seed=false
    Rotation Angle:NiagaraFloat=&RandomRangeFloat(Minimum=0, Maximum=1)
    Rotation Axis=<0 0 1>
    Rotation Mode=::NewEnumerator1
    Shape Primitive=::NewEnumerator1
    Surface Only=true
    UseOwnerScale=true
   )
   &AddVelocity(Velocity=<0 -35 100>, Velocity Speed Scale=&FN_FrontEndFloatModifier(Gameplay=0, Front End=1))
   &CylinderLocation(
    Cylinder Height=1
    Cylinder Radius=30
    MaskSpawnGroup=false
    Module.Enable Lathe Profile=false
    Module.Override Randomness=false
    Module.Surface Only=true
    Surface Only Band Thickness=1
    Use Endcaps In Surface Only Mode=false
   ) {enabled=false}
  ]
  particleUpdateScript [
   &ParticleState()
   &SpriteRotationRate(Rotation Rate=&Multiply_Float(
     A=&RandomRangeFloat(Minimum=5, Maximum=10)
     B=&SelectFloatFromArray(
      Array Sampling Mode=::NewEnumerator0
      Direct Array Index=?
      Float Selection Array=DataInterface() {propertyValues {FloatData [-1,1]}}
      Random Seed=0
     )
    ))
   &GravityForce(
    Coordinate Space=::Local
    Gravity=&MakeVector(
     X=0
     Y=&FN_FrontEndFloatModifier(
      Gameplay=0
      Front End=&RemapRange(Input Value To Remap=$System.Timeline002.NormalizedTime, Input Min=0, Input Max=1, Output Min=0, Output Max=-50)
     )
     Z=&FN_FrontEndFloatModifier(
      Gameplay=0
      Front End=&RemapRange(Input Value To Remap=$System.Timeline002.NormalizedTime, Input Min=0, Input Max=1, Output Min=0, Output Max=150)
     )
    )
   )
   &ScaleSpriteSize(Scale Factor=&Vector2DFromFloat(Value=&FloatFromCurve(FloatCurve=DataInterface() {propertyValues {Curve {keys [
          {arriveTangent=1.3248, interpMode=RCIM_Cubic, leaveTangent=1.3248, tangentMode=RCTM_User, time=0, value=0.7}
          {arriveTangent=0.4626, interpMode=RCIM_Cubic, leaveTangent=0.4626, tangentMode=RCTM_User, time=0.9978, value=1.202}
         ]}}}))) {enabled=false}
   &Drag(Drag=3, Use Linear Drag=true, Use Rotational Drag=true) {enabled=false}
   &InheritSourceMovement(
    Applied Position Scale=&VectorFromFloat(Value=&FloatFromCurve(FloatCurve=DataInterface() {propertyValues {Curve {keys [
          {arriveTangent=-0.3402, interpMode=RCIM_Cubic, leaveTangent=-0.3402, tangentMode=RCTM_User, time=0, value=1}
          {arriveTangent=-1.6874, interpMode=RCIM_Cubic, leaveTangent=-1.6874, tangentMode=RCTM_User, time=1, value=0.2}
         ]}}}))
    ApplyForce=false
    ApplyPosition=true
    ApplyPositionFalloff=false
   )
   &SolveForcesAndVelocity(Clamp Velocity=false, Limit Acceleration=false)
   &ScaleColor(
    Scale Alpha=&FloatFromCurve(FloatCurve=DataInterface() {propertyValues {Curve {keys [
         {arriveTangent=1.4796, interpMode=RCIM_Cubic, leaveTangent=1.4796, tangentMode=RCTM_User, time=0, value=0}
         {arriveTangent=-0.2124, interpMode=RCIM_Cubic, leaveTangent=-0.2124, tangentMode=RCTM_User, time=0.5, value=1}
         {arriveTangent=-2.9378, interpMode=RCIM_Cubic, leaveTangent=-2.9378, tangentMode=RCTM_User, time=1, value=0}
        ]}}})
    ScaleA=true
    ScaleRGB=true
   )
   &DynamicMaterialParameters(
    Index 0 Param 1=&Add_Float(
     A=$Particles.RandomFloat
     B=&FloatFromCurve(
      FloatCurve=DataInterface() {propertyValues {Curve {keys [
          {arriveTangent=1.0192, interpMode=RCIM_Cubic, leaveTangent=1.0192, tangentMode=RCTM_User, time=0, value=1}
          {arriveTangent=0.2834, interpMode=RCIM_Cubic, leaveTangent=0.2834, tangentMode=RCTM_User, time=1, value=1.5}
         ]}}}
      CurveIndex=&Add_Float(A=&Multiply_Float(A=$Particles.Initial.MaterialRandom, B=0.5), B=$Particles.NormalizedAge)
      Scale Curve=0.8
     )
    )
    Index 0 Param 3=&FloatFromCurve(FloatCurve=DataInterface() {propertyValues {Curve {keys [
         {arriveTangent=-53.9199, interpMode=RCIM_Cubic, leaveTangent=-8.9727, tangentMode=RCTM_User, time=0, value=1}
         {arriveTangent=0.2347, interpMode=RCIM_Cubic, leaveTangent=0.2723, tangentMode=RCTM_Break, time=0.25, value=0}
        ]}}})
    Index 0 Param 4=&FloatFromCurve(FloatCurve=DataInterface() {propertyValues {Curve {keys [<at 0.5 1>,<at 1 0>]}}})
    Index 2 Param 1=&FloatFromCurve(FloatCurve=DataInterface() {propertyValues {Curve {keys [{interpMode=RCIM_Cubic, leaveTangent=1.4301, tangentMode=RCTM_Break, time=0, value=0}, <at 1 1>]}}})
    Index 2 Param 2=&FloatFromCurve(
     FloatCurve=DataInterface() {propertyValues {Curve {keys [
         {arriveTangent=0.6777, interpMode=RCIM_Cubic, leaveTangent=0.6777, tangentMode=RCTM_User, time=0, value=0.1}
         {arriveTangent=0.0259, interpMode=RCIM_Cubic, leaveTangent=0.0259, tangentMode=RCTM_User, time=1, value=0.4}
        ]}}}
     Scale Curve=&RandomRangeFloat(Minimum=0.2, Maximum=0.5)
    )
    Index 2 Param 3=&FloatFromCurve(FloatCurve=DataInterface() {propertyValues {Curve {keys [
         {interpMode=RCIM_Cubic, leaveTangent=2.3064, tangentMode=RCTM_Break, time=0, value=1}
         {interpMode=RCIM_Cubic, leaveTangent=11.1233, tangentMode=RCTM_Break, time=1, value=2}
        ]}}})
    Index 2 Param 4=&RandomRangeFloat(Minimum=0.4, Maximum=1.1)
    Module.Param0WriteEnabled=true
    Module.Param0WriteEnabled2=true
    Module.Param1WriteEnabled=false
    Module.Param1WriteEnabled2=true
    Module.Param2WriteEnabled=true
    Module.Param2WriteEnabled2=true
    Module.Param3WriteEnabled=true
    Module.Param3WriteEnabled2=true
    Write Parameter Index 2=true
   )
   &CameraOffset(Camera Offset Amount=35)
   &KillParticles(Kill Particles=&SetBoolByFloatComparison(
     A=&MakeFloatFromVector(
      Vector=&ConvertVectorToPosition(
       Input Position=Linked() {linkedVariable {name=Particles.Position, type=NiagaraPosition}}
       Conversion Operation:ENiagara_LWCConvertVecToPos=::NewEnumerator1
      )
      Channel=::NewEnumerator2
     )
     B=8000
     Comparison Type=::NewEnumerator6
    )) {enabled=false}
  ]
  rendererClasses [&NiagaraSpriteRendererProperties]
  renderers [
   {rendererClass=&NiagaraSpriteRendererProperties}
   {rendererClass=&NiagaraSpriteRendererProperties, rendererIndex=1}
  ]
 }
 Cloud_FE {
  bEnabled=false
  emitterUpdateScript [
   &EmitterState(
    Enable Distance Culling=true
    Life Cycle Mode=::NewEnumerator1
    Loop Duration=2
    Max Distance Response=::SleepAndClearParticles
    Module.bApplySpawnCountScaleByDistance=false
    Module.bUseMaxDistance=true
    Module.bUseMinDistance=false
    Reset Age On Awaken=true
    Scalability Mode=::NewEnumerator1
    Scale Spawn Count=true
    Spawn Count Scale By Distance=DataInterface() {propertyValues {Curve {keys [<at 0 0>,<at 0.3 0.95>,<at 0.75 1>,<at 1 0>]}}}
    UseLoopDelay=false
   )
   &SpawnRate(
    Spawn Probability=&FN_FrontEndFloatModifier(Gameplay=0, Front End=1)
    SpawnRate=&Lerp_Float(A=1.5, B=3, Alpha=$User.ParaGlideLeanAlpha)
    Use Spawn Probability=true
   )
  ]
  particleSpawnScript [
   /SetVariables(Particles.MeshIndex=&RandomRangeInt(Minimum=0, Maximum=2)) {isSetParameters=true, module=SetVariables_#8}
   /Niagara/Modules/Spawn/Initialization/V2/InitializeParticle(
    Color=&ColorFromCurve(
     ColorCurve=DataInterface() {propertyValues {
       AlphaCurve {keys [<at 0 1 RCIM_Cubic>,<at 1 1 RCIM_Cubic>]}
       BlueCurve {keys [<at 0 1 RCIM_Cubic>,<at 0.7008 1>,<at 1 1 RCIM_Cubic>]}
       GreenCurve {keys [<at 0 1 RCIM_Cubic>,<at 0.7008 0.8409>,<at 1 0.9 RCIM_Cubic>]}
       RedCurve {keys [<at 0 1 RCIM_Cubic>,<at 0.7008 0.8>,<at 1 0.9473 RCIM_Cubic>]}
      }}
     CurveIndex=&RandomRangeFloat(Minimum=0, Maximum=1)
    )
    Lifetime Max=2.5
    Lifetime Min=2
    Lifetime Mode=::NewEnumerator1
    Mesh Scale=&Multiply_Vector(
     A=&RandomRangeVector(Minimum=<0.6 0.8 0.8>, Maximum=<0.8 1 1>)
     B=&VectorFromFloat(Value=&RandomRangeFloat(Minimum=0.75, Maximum=1.25))
    )
    Mesh Scale Mode=::NewEnumerator4
    UsePositionOffset=false
    Write Material Random=true
   ) {module=InitializeParticle001}
   &InitialMeshOrientation(
    Mesh Orientation Mode=::NewEnumerator7
    Rotation=&SelectVectorFromArray(
     Array Sampling Mode=::NewEnumerator0
     Direct Array Index=?
     Vector Selection Array=DataInterface() {propertyValues {FloatData [<0.05 0 0.5>, <0 0 0>]}}
     Random Seed=0
    )
    Use Rotation Vector=true
   )
   &ShapeLocation(
    Cylinder Height=200
    Cylinder Radius=50
    MaskSpawnGroup=false
    Offset=&Lerp_Vector(A=<0 50 0>, B=<0 125 -25>, Alpha=$User.ParaGlideLeanAlpha)
    Offset Coordinate Space=::Local
    Offset Mode=::NewEnumerator0
    Override Seed=false
    Shape Primitive=::NewEnumerator1
    Surface Only=false
    UseOwnerScale=true
   )
   &AddVelocity(Velocity=&Lerp_Vector(A=<0 -25 75>, B=<1 -150 150>, Alpha=$User.ParaGlideLeanAlpha))
  ]
  particleUpdateScript [
   &ParticleState()
   &ScaleMeshSize(Scale Factor=&MakeVector(
     X=&Lerp_Float(A=1, B=1.1, Alpha=$User.ParaGlideLeanAlpha)
     Y=1
     Z=&Lerp_Float(A=1, B=1.1, Alpha=$User.ParaGlideLeanAlpha)
    )) {module=ScaleMeshSize001}
   &ScaleMeshSize(Scale Factor=&ScalabilityDistanceBasedVector(
     Curve:NiagaraDataInterfaceVectorCurve=DataInterface() {propertyValues {
       XCurve {keys [{tangentMode=RCTM_User, time=0, value=1}, {tangentMode=RCTM_User, time=1, value=0}]}
       YCurve {keys [{tangentMode=RCTM_User, time=0, value=1}, {tangentMode=RCTM_User, time=1, value=0}]}
       ZCurve {keys [{tangentMode=RCTM_User, time=0, value=1}, {tangentMode=RCTM_User, time=1, value=0}]}
      }}
     CurveIndex:NiagaraFloat=$Emitter.Scalability.DistanceFraction
    ))
   &OrientMeshToVector(
    Look At Direction:Vector3f=&ReturnCameraVector(Camera Vectors=::NewEnumerator3, Coordinate Space=::Simulation)
    Override Quat=false
    UseOptionalSideAxis=false
    UseOptionalUpAxis=false
   )
   &ScaleColor(
    Scale Alpha=&FloatFromCurve(
     FloatCurve=DataInterface() {propertyValues {Curve {keys [
         {interpMode=RCIM_Cubic, tangentMode=RCTM_User, time=0.1, value=0}
         {interpMode=RCIM_Cubic, tangentMode=RCTM_User, time=0.5, value=1}
         <at 0.9 0>
        ]}}}
     Scale Curve=&FloatFromCurve(
      FloatCurve=DataInterface() {propertyValues {Curve {keys [
          {arriveTangent=-0.0101, interpMode=RCIM_Cubic, leaveTangent=-0.0101, tangentMode=RCTM_User, time=0, value=0.75}
          {arriveTangent=3.0014, interpMode=RCIM_Cubic, leaveTangent=3.0014, tangentMode=RCTM_User, time=0.1813, value=0.9875}
          {arriveTangent=1, time=0.2, value=1}
          <at 1 1>
         ]}}}
      CurveIndex=$Emitter.Scalability.DistanceFraction
     )
    )
    ScaleA=true
    ScaleRGB=true
   )
   &DynamicMaterialParameters(
    Index 0 Param 1=&RandomRangeFloat(Minimum=2, Maximum=3)
    Index 0 Param 2=&RandomRangeFloat(Minimum=2, Maximum=3)
    Index 0 Param 3=1
    Index 0 Param 4=1
    Module.Param0WriteEnabled=true
    Module.Param1WriteEnabled=true
    Module.Param2WriteEnabled=true
    Module.Param3WriteEnabled=true
   )
   &InheritSourceMovement(
    Applied Position Scale=&VectorFromFloat(Value=&FloatFromCurve(FloatCurve=DataInterface() {propertyValues {Curve {keys [
          {arriveTangent=-0.3402, interpMode=RCIM_Cubic, leaveTangent=-0.3402, tangentMode=RCTM_User, time=0, value=1}
          {arriveTangent=-1.6874, interpMode=RCIM_Cubic, leaveTangent=-1.6874, tangentMode=RCTM_User, time=1, value=0.2}
         ]}}}))
    ApplyForce=false
    ApplyPosition=true
    ApplyPositionFalloff=false
   )
   &SolveForcesAndVelocity(Clamp Velocity=false, Limit Acceleration=false)
  ]
  rendererClasses [&NiagaraMeshRendererProperties]
  renderers [
   {rendererClass=&NiagaraMeshRendererProperties, rendererIndex=0}
   {rendererClass=&NiagaraMeshRendererProperties, rendererIndex=1}
   {rendererClass=&NiagaraMeshRendererProperties, rendererIndex=2}
   {rendererClass=&NiagaraMeshRendererProperties, rendererIndex=3}
  ]
 }
 Cloud_FE_Low {
  bEnabled=false
  emitterUpdateScript [
   &EmitterState(
    Enable Distance Culling=true
    Life Cycle Mode=::NewEnumerator1
    Loop Duration=2
    Max Distance Response=::SleepAndClearParticles
    Module.bApplySpawnCountScaleByDistance=false
    Module.bUseMaxDistance=true
    Module.bUseMinDistance=false
    Reset Age On Awaken=true
    Scalability Mode=::NewEnumerator1
    Scale Spawn Count=true
    Spawn Count Scale By Distance=DataInterface() {propertyValues {Curve {keys [<at 0 0>,<at 0.3 0.95>,<at 0.75 1>,<at 1 0>]}}}
    UseLoopDelay=false
   )
   &SpawnRate(
    Spawn Probability=&FN_FrontEndFloatModifier(Gameplay=0, Front End=1)
    SpawnRate=&Lerp_Float(A=1, B=2, Alpha=$User.ParaGlideLeanAlpha)
    Use Spawn Probability=true
   )
  ]
  particleSpawnScript [
   /SetVariables(Particles.MeshIndex=&RandomRangeInt(Minimum=0, Maximum=2)) {isSetParameters=true, module=SetVariables_#8}
   &InitializeParticle(
    Color=&ColorFromCurve(
     ColorCurve=DataInterface() {propertyValues {
       AlphaCurve {keys [<at 0 1 RCIM_Cubic>,<at 1 1 RCIM_Cubic>]}
       BlueCurve {keys [<at 0 1 RCIM_Cubic>,<at 0.7008 1>,<at 1 1 RCIM_Cubic>]}
       GreenCurve {keys [<at 0 1 RCIM_Cubic>,<at 0.7008 0.8409>,<at 1 0.9 RCIM_Cubic>]}
       RedCurve {keys [<at 0 1 RCIM_Cubic>,<at 0.7008 0.8>,<at 1 0.9473 RCIM_Cubic>]}
      }}
     CurveIndex=&RandomRangeFloat(Minimum=0, Maximum=1)
    )
    Lifetime Max=2.5
    Lifetime Min=2
    Lifetime Mode=::NewEnumerator1
    Mesh Scale=&Multiply_Vector(
     A=&RandomRangeVector(Minimum=<0.6 0.8 0.8>, Maximum=<0.8 1 1>)
     B=&VectorFromFloat(Value=&RandomRangeFloat(Minimum=0.75, Maximum=1.25))
    )
    Mesh Scale Mode=::NewEnumerator4
    UsePositionOffset=false
    Write Material Random=true
   ) {module=InitializeParticle001}
   &InitialMeshOrientation(
    Mesh Orientation Mode=::NewEnumerator7
    Rotation=&SelectVectorFromArray(
     Array Sampling Mode=::NewEnumerator0
     Direct Array Index=?
     Vector Selection Array=DataInterface() {propertyValues {FloatData [<0.05 0 0.5>, <0 0 0>]}}
     Random Seed=0
    )
    Use Rotation Vector=true
   )
   &ShapeLocation(
    Cylinder Height=200
    Cylinder Radius=50
    MaskSpawnGroup=false
    Offset=&Lerp_Vector(A=<0 50 0>, B=<0 125 -25>, Alpha=$User.ParaGlideLeanAlpha)
    Offset Coordinate Space=::Local
    Offset Mode=::NewEnumerator0
    Override Seed=false
    Shape Primitive=::NewEnumerator1
    Surface Only=false
    UseOwnerScale=true
   )
   &AddVelocity(Velocity=&Lerp_Vector(A=<0 -25 75>, B=<1 -150 150>, Alpha=$User.ParaGlideLeanAlpha))
  ]
  particleUpdateScript [
   &ParticleState()
   &ScaleMeshSize(Scale Factor=&MakeVector(
     X=&Lerp_Float(A=1, B=1.1, Alpha=$User.ParaGlideLeanAlpha)
     Y=1
     Z=&Lerp_Float(A=1, B=1.1, Alpha=$User.ParaGlideLeanAlpha)
    )) {module=ScaleMeshSize001}
   &ScaleMeshSize(Scale Factor=&ScalabilityDistanceBasedVector(
     Curve:NiagaraDataInterfaceVectorCurve=DataInterface() {propertyValues {
       XCurve {keys [{tangentMode=RCTM_User, time=0, value=1}, {tangentMode=RCTM_User, time=1, value=0}]}
       YCurve {keys [{tangentMode=RCTM_User, time=0, value=1}, {tangentMode=RCTM_User, time=1, value=0}]}
       ZCurve {keys [{tangentMode=RCTM_User, time=0, value=1}, {tangentMode=RCTM_User, time=1, value=0}]}
      }}
     CurveIndex:NiagaraFloat=$Emitter.Scalability.DistanceFraction
    ))
   &OrientMeshToVector(
    Look At Direction:Vector3f=&ReturnCameraVector(Camera Vectors=::NewEnumerator3, Coordinate Space=::Simulation)
    Override Quat=false
    UseOptionalSideAxis=false
    UseOptionalUpAxis=false
   )
   &ScaleColor(
    Scale Alpha=&FloatFromCurve(
     FloatCurve=DataInterface() {propertyValues {Curve {keys [
         {interpMode=RCIM_Cubic, tangentMode=RCTM_User, time=0.1, value=0}
         {interpMode=RCIM_Cubic, tangentMode=RCTM_User, time=0.5, value=1}
         <at 0.9 0>
        ]}}}
     Scale Curve=&FloatFromCurve(
      FloatCurve=DataInterface() {propertyValues {Curve {keys [
          {arriveTangent=-0.0101, interpMode=RCIM_Cubic, leaveTangent=-0.0101, tangentMode=RCTM_User, time=0, value=0.75}
          {arriveTangent=3.0014, interpMode=RCIM_Cubic, leaveTangent=3.0014, tangentMode=RCTM_User, time=0.1813, value=0.9875}
          {arriveTangent=1, time=0.2, value=1}
          <at 1 1>
         ]}}}
      CurveIndex=$Emitter.Scalability.DistanceFraction
      Scale Curve=2
     )
    )
    ScaleA=true
    ScaleRGB=true
   )
   &DynamicMaterialParameters(
    Index 0 Param 1=&RandomRangeFloat(Minimum=2, Maximum=3)
    Index 0 Param 2=&RandomRangeFloat(Minimum=2, Maximum=3)
    Index 0 Param 3=1
    Index 0 Param 4=1
    Module.Param0WriteEnabled=true
    Module.Param1WriteEnabled=true
    Module.Param2WriteEnabled=true
    Module.Param3WriteEnabled=true
   )
   &InheritSourceMovement(
    Applied Position Scale=&VectorFromFloat(Value=&FloatFromCurve(FloatCurve=DataInterface() {propertyValues {Curve {keys [
          {arriveTangent=-0.3402, interpMode=RCIM_Cubic, leaveTangent=-0.3402, tangentMode=RCTM_User, time=0, value=1}
          {arriveTangent=-1.6874, interpMode=RCIM_Cubic, leaveTangent=-1.6874, tangentMode=RCTM_User, time=1, value=0.2}
         ]}}}))
    ApplyForce=false
    ApplyPosition=true
    ApplyPositionFalloff=false
   )
   &SolveForcesAndVelocity(Clamp Velocity=false, Limit Acceleration=false)
  ]
  rendererClasses [&NiagaraMeshRendererProperties]
  renderers [
   {rendererClass=&NiagaraMeshRendererProperties, rendererIndex=0}
   {rendererClass=&NiagaraMeshRendererProperties, rendererIndex=1}
  ]
 }
 Cloud_Low {
  bEnabled=false
  emitterUpdateScript [
   &EmitterState(
    Enable Distance Culling=true
    Life Cycle Mode=::NewEnumerator1
    Loop Duration=2
    Max Distance Response=::SleepAndClearParticles
    Module.bApplySpawnCountScaleByDistance=false
    Module.bUseMaxDistance=true
    Module.bUseMinDistance=false
    Reset Age On Awaken=true
    Scalability Mode=::NewEnumerator1
    Scale Spawn Count=true
    Spawn Count Scale By Distance=DataInterface() {propertyValues {Curve {keys [<at 0 0>,<at 0.3 0.95>,<at 0.75 1>,<at 1 0>]}}}
    UseLoopDelay=false
   )
   &SpawnRate(
    Spawn Probability=&FN_FrontEndFloatModifier(Gameplay=1, Front End=0)
    SpawnRate=&Lerp_Float(A=1.5, B=3, Alpha=$User.ParaGlideLeanAlpha)
    Use Spawn Probability=true
   )
  ]
  particleSpawnScript [
   /SetVariables(Particles.MeshIndex=&RandomRangeInt(Minimum=0, Maximum=2)) {isSetParameters=true, module=SetVariables_#8}
   &InitializeParticle(
    Color=&ColorFromCurve(
     ColorCurve=DataInterface() {propertyValues {
       AlphaCurve {keys [<at 0 1 RCIM_Cubic>,<at 1 1 RCIM_Cubic>]}
       BlueCurve {keys [<at 0 1 RCIM_Cubic>,<at 0.7008 1>,<at 1 1 RCIM_Cubic>]}
       GreenCurve {keys [<at 0 1 RCIM_Cubic>,<at 0.7008 0.8409>,<at 1 0.9 RCIM_Cubic>]}
       RedCurve {keys [<at 0 1 RCIM_Cubic>,<at 0.7008 0.8>,<at 1 0.9473 RCIM_Cubic>]}
      }}
     CurveIndex=&RandomRangeFloat(Minimum=0, Maximum=1)
    )
    Lifetime Max=2.5
    Lifetime Min=2
    Lifetime Mode=::NewEnumerator1
    Mesh Scale=&Multiply_Vector(
     A=&RandomRangeVector(Minimum=<0.6 0.8 0.8>, Maximum=<0.8 1 1>)
     B=&VectorFromFloat(Value=&RandomRangeFloat(Minimum=1, Maximum=1.5))
    )
    Mesh Scale Mode=::NewEnumerator4
    UsePositionOffset=false
    Write Material Random=true
   ) {module=InitializeParticle001}
   &InitialMeshOrientation(
    Mesh Orientation Mode=::NewEnumerator7
    Rotation=&SelectVectorFromArray(
     Array Sampling Mode=::NewEnumerator0
     Direct Array Index=?
     Vector Selection Array=DataInterface() {propertyValues {FloatData [<0.05 0 0.5>, <0 0 0>]}}
     Random Seed=0
    )
    Use Rotation Vector=true
   )
   &ShapeLocation(
    Disc Coverage=-1
    MaskSpawnGroup=false
    Offset=&Lerp_Vector(A=<0 0 0>, B=<0 100 -250>, Alpha=$User.ParaGlideLeanAlpha)
    Offset Coordinate Space=::Local
    Offset Mode=::NewEnumerator0
    Override Seed=false
    Ring Radius=&RandomRangeFloat(Minimum=80, Maximum=100)
    Shape Primitive=::NewEnumerator4
    UseOwnerScale=true
   )
  ]
  particleUpdateScript [
   &ParticleState()
   &ScaleMeshSize(Scale Factor=&MakeVector(
     X=&Lerp_Float(A=1, B=1.1, Alpha=$User.ParaGlideLeanAlpha)
     Y=1
     Z=&Lerp_Float(A=1, B=1.1, Alpha=$User.ParaGlideLeanAlpha)
    )) {module=ScaleMeshSize001}
   &ScaleMeshSize(Scale Factor=&ScalabilityDistanceBasedVector(
     Curve:NiagaraDataInterfaceVectorCurve=DataInterface() {propertyValues {
       XCurve {keys [{tangentMode=RCTM_User, time=0, value=1}, {tangentMode=RCTM_User, time=1, value=0}]}
       YCurve {keys [{tangentMode=RCTM_User, time=0, value=1}, {tangentMode=RCTM_User, time=1, value=0}]}
       ZCurve {keys [{tangentMode=RCTM_User, time=0, value=1}, {tangentMode=RCTM_User, time=1, value=0}]}
      }}
     CurveIndex:NiagaraFloat=$Emitter.Scalability.DistanceFraction
    ))
   &OrientMeshToVector(
    Look At Direction:Vector3f=&ReturnCameraVector(Camera Vectors=::NewEnumerator3, Coordinate Space=::Simulation)
    Override Quat=false
    UseOptionalSideAxis=false
    UseOptionalUpAxis=false
   )
   &ScaleColor(
    Scale Alpha=&FloatFromCurve(
     FloatCurve=DataInterface() {propertyValues {Curve {keys [
         {interpMode=RCIM_Cubic, tangentMode=RCTM_User, time=0.1, value=0}
         {interpMode=RCIM_Cubic, tangentMode=RCTM_User, time=0.5, value=1}
         <at 0.9 0>
        ]}}}
     Scale Curve=&FloatFromCurve(
      FloatCurve=DataInterface() {propertyValues {Curve {keys [
          {arriveTangent=-0.0101, interpMode=RCIM_Cubic, leaveTangent=-0.0101, tangentMode=RCTM_User, time=0, value=0.75}
          {arriveTangent=3.0014, interpMode=RCIM_Cubic, leaveTangent=3.0014, tangentMode=RCTM_User, time=0.1813, value=0.9875}
          {arriveTangent=1, time=0.2, value=1}
          <at 1 1>
         ]}}}
      CurveIndex=$Emitter.Scalability.DistanceFraction
      Scale Curve=2
     )
    )
    ScaleA=true
    ScaleRGB=true
   )
   &DynamicMaterialParameters(
    Index 0 Param 1=&RandomRangeFloat(Minimum=2, Maximum=3)
    Index 0 Param 2=&RandomRangeFloat(Minimum=2, Maximum=3)
    Index 0 Param 3=1
    Index 0 Param 4=1
    Module.Param0WriteEnabled=true
    Module.Param1WriteEnabled=true
    Module.Param2WriteEnabled=true
    Module.Param3WriteEnabled=true
   )
   &InheritSourceMovement(
    Applied Position Scale=&VectorFromFloat(Value=&FloatFromCurve(FloatCurve=DataInterface() {propertyValues {Curve {keys [
          {arriveTangent=-0.3402, interpMode=RCIM_Cubic, leaveTangent=-0.3402, tangentMode=RCTM_User, time=0, value=1}
          {arriveTangent=-1.6874, interpMode=RCIM_Cubic, leaveTangent=-1.6874, tangentMode=RCTM_User, time=1, value=0.2}
         ]}}}))
    ApplyForce=false
    ApplyPosition=true
    ApplyPositionFalloff=false
   )
   &SolveForcesAndVelocity(Clamp Velocity=false, Limit Acceleration=false)
  ]
  rendererClasses [&NiagaraMeshRendererProperties]
  renderers [
   {rendererClass=&NiagaraMeshRendererProperties, rendererIndex=0}
   {rendererClass=&NiagaraMeshRendererProperties, rendererIndex=1}
  ]
 }
 Cloud_Wispy {
  emitterUpdateScript [
   &EmitterState(Inactive Response=::NewEnumerator1, Life Cycle Mode=::NewEnumerator1, UseLoopDelay=false)
   &SpawnBurst_Instantaneous(
    Spawn Count=&MakeCustomIntFromBool(Bool=$System.Spawning, True INT=5, False INT=0)
    Use Spawn Probability=false
    UseLoopCountLimit=false
   )
   &SpawnBurst_Instantaneous(
    Spawn Count=&MakeCustomIntFromBool(Bool=$System.Spawning, True INT=7, False INT=0)
    Spawn Group=1
    Spawn Time=2
    Use Spawn Probability=false
    UseLoopCountLimit=false
   ) {module=SpawnBurst_Instantaneous001}
   &SpawnBurst_Instantaneous(
    Spawn Count=&MakeCustomIntFromBool(Bool=$System.Spawning, True INT=8, False INT=0)
    Spawn Group=2
    Spawn Time=3
    Use Spawn Probability=false
    UseLoopCountLimit=false
   ) {module=SpawnBurst_Instantaneous002}
   &SpawnBurst_Instantaneous(
    Spawn Count=&MakeCustomIntFromBool(Bool=$System.Spawning, True INT=5, False INT=0)
    Spawn Group=3
    Spawn Time=5
    Use Spawn Probability=false
    UseLoopCountLimit=false
   ) {module=SpawnBurst_Instantaneous003}
   &SpawnBurst_Instantaneous(
    Spawn Count=&MakeCustomIntFromBool(Bool=$System.Spawning, True INT=6, False INT=0)
    Spawn Group=4
    Spawn Time=7
    Use Spawn Probability=false
    UseLoopCountLimit=false
   ) {module=SpawnBurst_Instantaneous004}
   &SpawnBurst_Instantaneous(
    Spawn Count=&MakeCustomIntFromBool(Bool=$System.Spawning, True INT=8, False INT=0)
    Spawn Group=5
    Spawn Time=8.5
    Use Spawn Probability=false
    UseLoopCountLimit=false
   ) {module=SpawnBurst_Instantaneous005}
   &SpawnRate(SpawnRate=3, Use Spawn Probability=false) {enabled=false}
  ]
  particleSpawnScript [
   &InitializeParticle(
    Color=<rgba 1.5 1.3724 1.4638 0.35>
    Lifetime=5
    Sprite Rotation Angle=&RandomRangeFloat(Minimum=-360, Maximum=360)
    Sprite Rotation Mode=::NewEnumerator0
    Sprite Size Mode=::NewEnumerator3
    Uniform Sprite Size=&Multiply_Float(
     A=&RandomRangeFloat(Minimum=40, Maximum=50)
     B=&SelectFloatFromArray(Direct Array Index=$Emitter.SpawnGroup, Float Selection Array=DataInterface() {propertyValues {FloatData [1,1.3,1.5,1,1.3,1.5]}})
    )
    UsePositionOffset=false
    Write Material Random=true
   ) {module=InitializeParticle001}
   /Niagara/Modules/Spawn/Initialization/InitializeParticle(
    Color=&UniformRangedLinearColor(Minimum=<rgba 1.5 1.3724 1.4638 1>, Maximum=<rgba 0.2913 0.3405 0.526 1>)
    Lifetime=&UniformRangedFloat(Minimum=10, Maximum=12)
    Module.Write Color=true
    Module.Write Lifetime=true
    Module.Write Mass=true
    Module.Write Material Random=true
    Module.Write Position=true
    Module.Write Scale=false
    Module.Write SpriteRotation=true
    Module.Write SpriteSize=true
    Sprite Rotation=&NormalizedAngleToDegrees(NormalizedAngle=&UniformRangedFloat(Minimum=0, Maximum=360))
    Sprite Size=&Vector2DFromFloat(Value=&RandomRangeFloat(Minimum=35, Maximum=45))
   ) {enabled=false}
   /SetVariables(Particles.Position=&MaskVectorBySpawnGroup(
     Masked Vector=&Add_Vector(A=Linked() {linkedVariable {name=Engine.Owner.Position, type=NiagaraPosition}}, B=<0 0 10>)
     Passthrough Vector=Linked() {linkedVariable {name=Particles.Position, type=NiagaraPosition}}
     Spawn Group Mask=1
    )) {enabled=false, isSetParameters=true, module=SetVariables_#9}
   &CameraOffset(Camera Offset Amount=&RandomRangeFloat(Minimum=5, Maximum=10)) {module=CameraOffset001}
   &ShapeLocation(
    MaskSpawnGroup=false
    Offset=&FN_FrontEndVector3Modifier(Gameplay=<0 0 -300>, Front End=<0 0 -125>)
    Offset Mode=::NewEnumerator0
    Override Seed=false
    Ring / Disc Distribution Mode:ENiagaraSphereDistributionMode=::NewEnumerator1
    Ring Radius=200
    Shape Primitive=::NewEnumerator4
    U Position=&SelectFloatFromArray(Direct Array Index=$Emitter.SpawnGroup, Float Selection Array=DataInterface() {propertyValues {FloatData [0.1,0.5,0.3,0.8,0.2,0.9]}})
    UseOwnerScale=true
   ) {module=ShapeLocation001}
   &ShapeLocation(
    Cylinder Height=10
    Cylinder Radius=&SelectFloatFromArray(Direct Array Index=$Emitter.SpawnGroup, Float Selection Array=DataInterface() {propertyValues {FloatData [30,45,50,30,40,45]}})
    MaskSpawnGroup=false
    Non Uniform Scale=<1.5 0.7 0.7>
    Override Seed=false
    Rotation Angle:NiagaraFloat=&RandomRangeFloat(Minimum=0, Maximum=1)
    Rotation Axis=<0 0 1>
    Rotation Mode=::NewEnumerator1
    Shape Primitive=::NewEnumerator1
    Surface Only=true
    Use Endcaps In Surface Only Mode=false
    UseOwnerScale=true
   )
   &AddVelocity(Velocity=<0 -35 100>, Velocity Speed Scale=&FN_FrontEndFloatModifier(Gameplay=0, Front End=1))
   &CylinderLocation(
    Cylinder Height=1
    Cylinder Radius=30
    MaskSpawnGroup=false
    Module.Enable Lathe Profile=false
    Module.Override Randomness=false
    Module.Surface Only=true
    Surface Only Band Thickness=1
    Use Endcaps In Surface Only Mode=false
   ) {enabled=false}
   &SubUVAnimation(Start Frame=24, SubUV Animation Mode:ENiagara_SubUVLookupMode=::NewEnumerator2)
  ]
  particleUpdateScript [
   &ParticleState()
   &SpriteRotationRate(Rotation Rate=&Multiply_Float(
     A=&RandomRangeFloat(Minimum=10, Maximum=15)
     B=&SelectFloatFromArray(
      Array Sampling Mode=::NewEnumerator0
      Direct Array Index=?
      Float Selection Array=DataInterface() {propertyValues {FloatData [-1,1]}}
      Random Seed=0
     )
    ))
   &GravityForce(
    Coordinate Space=::Local
    Gravity=&MakeVector(
     X=0
     Y=&FN_FrontEndFloatModifier(
      Gameplay=0
      Front End=&RemapRange(Input Value To Remap=$System.Timeline002.NormalizedTime, Input Min=0, Input Max=1, Output Min=0, Output Max=-50)
     )
     Z=&FN_FrontEndFloatModifier(
      Gameplay=0
      Front End=&RemapRange(Input Value To Remap=$System.Timeline002.NormalizedTime, Input Min=0, Input Max=1, Output Min=0, Output Max=150)
     )
    )
   ) {module=GravityForce001}
   &ScaleSpriteSize(Scale Factor=&Vector2DFromFloat(Value=&FloatFromCurve(FloatCurve=DataInterface() {propertyValues {Curve {keys [
          {arriveTangent=1.3248, interpMode=RCIM_Cubic, leaveTangent=1.3248, tangentMode=RCTM_User, time=0, value=0.7}
          {arriveTangent=0.4626, interpMode=RCIM_Cubic, leaveTangent=0.4626, tangentMode=RCTM_User, time=0.9978, value=1.202}
         ]}}}))) {enabled=false}
   &Drag(Drag=3, Use Linear Drag=true, Use Rotational Drag=true) {enabled=false}
   &InheritSourceMovement(
    Applied Position Scale=&VectorFromFloat(Value=&FloatFromCurve(FloatCurve=DataInterface() {propertyValues {Curve {keys [
          {arriveTangent=-0.3402, interpMode=RCIM_Cubic, leaveTangent=-0.3402, tangentMode=RCTM_User, time=0, value=1}
          {arriveTangent=-1.6874, interpMode=RCIM_Cubic, leaveTangent=-1.6874, tangentMode=RCTM_User, time=1, value=0.2}
         ]}}}))
    ApplyForce=false
    ApplyPosition=true
    ApplyPositionFalloff=false
   )
   &SolveForcesAndVelocity(Clamp Velocity=false, Limit Acceleration=false)
   &ScaleColor(
    Scale Alpha=&FloatFromCurve(FloatCurve=DataInterface() {propertyValues {Curve {keys [
         {arriveTangent=1.4796, interpMode=RCIM_Cubic, leaveTangent=1.4796, tangentMode=RCTM_User, time=0, value=0}
         {arriveTangent=-0.2124, interpMode=RCIM_Cubic, leaveTangent=-0.2124, tangentMode=RCTM_User, time=0.5, value=1}
         {arriveTangent=-2.9378, interpMode=RCIM_Cubic, leaveTangent=-2.9378, tangentMode=RCTM_User, time=1, value=0}
        ]}}})
    ScaleA=true
    ScaleRGB=true
   )
   &CameraOffset(Camera Offset Amount=35)
   &DynamicMaterialParameters(
    Index 0 Param 1=&FloatFromCurve(FloatCurve=DataInterface() {propertyValues {Curve {keys [<at 0 1>,<at 1 0>]}}})
    Index 0 Param 2=&UniformRangedFloat(Minimum=20, Maximum=24)
    Index 0 Param 3=5
    Index 0 Param 4=&FloatFromCurve(FloatCurve=DataInterface() {propertyValues {Curve {keys [<at 0.0123 0.25>,<at 0.99 -0.3>,<at 1 -0.4>]}}})
    Module.Param0WriteEnabled=true
    Module.Param1WriteEnabled=true
    Module.Param2WriteEnabled=true
    Module.Param3WriteEnabled=true
   )
   &KillParticles(Kill Particles=&SetBoolByFloatComparison(
     A=&MakeFloatFromVector(
      Vector=&ConvertVectorToPosition(
       Input Position=Linked() {linkedVariable {name=Particles.Position, type=NiagaraPosition}}
       Conversion Operation:ENiagara_LWCConvertVecToPos=::NewEnumerator1
      )
      Channel=::NewEnumerator2
     )
     B=8000
     Comparison Type=::NewEnumerator6
    )) {enabled=false}
  ]
  rendererClasses [&NiagaraSpriteRendererProperties]
  renderers [{rendererClass=&NiagaraSpriteRendererProperties}]
 }
 Hand_Rocks {
  emitterUpdateScript [
   &EmitterState(Life Cycle Mode=::NewEnumerator1, Loop Behavior=::NewEnumerator1, Loop Duration Mode=::NewEnumerator1, UseLoopDelay=false)
   &SpawnBurst_Instantaneous(Spawn Count=2, Use Spawn Probability=false, UseLoopCountLimit=false)
  ]
  particleSpawnScript [
   &FN_SetSpawnGroup(Number of Spawn Groups=2)
   /Niagara/Modules/Spawn/Initialization/V2/InitializeParticle(
    Color=<rgba 1 1 1 0.6>
    Lifetime=1.5
    Mesh Scale=<0.3 0.15 0.15>
    Mesh Scale Mode=::NewEnumerator4
    UsePositionOffset=false
    Write Material Random=true
   )
   &DynamicMaterialParameters(
    Index 1 Param 2=0.1
    Module.Param0WriteEnabled1=false
    Module.Param1WriteEnabled1=true
    Module.Param2WriteEnabled1=false
    Module.Param3WriteEnabled1=false
    Write Parameter Index 0=false
    Write Parameter Index 1=true
   )
  ]
  particleUpdateScript [
   &ParticleState(
    Kill Particles When Lifetime Has Elapsed=false
    Let Infinitely Lived Particles Die When Emitter Deactivates=true
    Loop Particles Lifetime=true
   )
   &SkeletalMeshLocation(Bone / Socket Index=$Particles.SpawnGroup, Bone Sampling Mode=::NewEnumerator4, MaskSpawnGroup=false, Skeletal Mesh=$User.SkeletalMesh)
   &OrientMeshToVector(
    Look At Direction:Vector3f=&FN_FrontEndVector3Modifier(
     Gameplay=&NormalizeVector(
      Vector To Normalize=&TransformVector(Vector=$Engine.Owner.Velocity, Destination Space=::Local)
      Safe Normalize=true
      Safe Fallback Threshold=0
      Safe Fallback Vector=<0 0 0.001>
     )
     Front End=&TransformVector(Vector=&Lerp_Vector(A=<0 0.4 -1>, B=<0 2 -1>, Alpha=$User.ParaGlideLeanAlpha), Destination Space=::Simulation)
    )
    Override Incoming Orientation=1
    Override Quat=false
    Reference Axis Vector:Vector3f=<-1 0 0>
    UseOptionalSideAxis=false
    UseOptionalUpAxis=false
   )
   /SetVariables() {isSetParameters=true}
   /SetVariables(Particles.CameraOffset_Bowl=3) {isSetParameters=true}
   &SolveForcesAndVelocity(Clamp Velocity=false, Limit Acceleration=false)
  ]
  rendererClasses [&NiagaraMeshRendererProperties]
  renderers [{rendererClass=&NiagaraMeshRendererProperties, rendererIndex=0}]
 }
 Ribbons_Light_Chroma {
  emitterUpdateScript [
   &EmitterState(Life Cycle Mode=::NewEnumerator1, Loop Duration=0, UseLoopDelay=false)
   /Game/Effects/Niagara/Modules/Skydiving/FN_DivingAttributes/FN_EmitterDivingAttributes_v1(Bone Index=1, SkeletalMesh:NiagaraDataInterfaceSkeletalMesh=$User.SkeletalMesh)
   &FN_SpawnRibbon_v2(Number of Ribbons=2, Spawn Mode=::NewEnumerator0, SpawnRate=64)
  ]
  particleSpawnScript [
   &FN_SetSpawnGroup(Number of Spawn Groups=2)
   &FN_InitializeRibbon_v3(
    Color:LinearColor=<rgba 0.5 0.5 0.5 0.6>
    Enable Skeletal Position=true
    Ribbon Width=40
    Sampled Position Coordinate Space:ENiagaraCoordinateSpace=::Local
    Sampled Rotation Coordinate Space:ENiagaraCoordinateSpace=::Local
    Sampled Velocity Coordinate Space:ENiagaraCoordinateSpace=::Local
   ) {module=FN_InitializeRibbon_v3001}
   &FN_InitializeRibbon_v3(
    Color:LinearColor=<rgba 0.4583 0.4583 0.4583 0.25>
    Custom Offset=true
    Enable Skeletal Position=true
    Lifetime=0.25
    Local Vector Offset:Vector3f=<0 1 0>
    Ribbon Width=25
   ) {enabled=false}
   &AddVelocity(
    Scale Added Velocity=&Lerp_Vector(A=<1 1 1>, B=<4 4 1>, Alpha=$System.Dive)
    Velocity=&FN_FrontEndVector3Modifier(Gameplay=<0 0 999>, Front End=<0 -333 333>)
   ) {enabled=false}
   &AddVelocity(Coordinate Space=::World, Velocity=<0 0 410>) {module=AddVelocity001}
   &AddVelocity(Velocity=&MakeCustomVectorFromBool(
     Bool=$Particles.Locked
     True Vector=<0 0 0>
     False Vector=&FN_FrontEndVector3Modifier(Gameplay=<0 0 500>, Front End=&Lerp_Vector(A=<0 -350 500>, B=<0 -1200 250>, Alpha=$User.ParaGlideLeanAlpha))
    )) {module=AddVelocity002}
   &DynamicMaterialParameters(
    Index 0 Param 1=&MaskFloatBySpawnGroup(Masked Float=0, Passthrough Float=0.33)
    Module.Param0WriteEnabled=true
    Module.Param1WriteEnabled=false
    Module.Param2WriteEnabled=false
    Module.Param3WriteEnabled=false
   )
   &SkeletalMeshLocation(
    Apply Sampled Orientation to Mesh Particles=true
    Bone / Socket Index=$Particles.SpawnGroup
    Bone Sampling Mode=::NewEnumerator4
    Inherit Velocity=false
    Interpolate Mesh Sampling=false
    MaskSpawnGroup=false
    Sampled Position Offset Coordinate Space:ENiagaraMeshTransforms=::NewEnumerator3
    Skeletal Mesh=$User.SkeletalMesh
   )
  ]
  particleUpdateScript [
   &InheritSourceMovement(Applied Position Scale=<0 0 0>, ApplyForce=false, ApplyPosition=true, ApplyPositionFalloff=false)
   &SolveForcesAndVelocity(Clamp Velocity=false, Limit Acceleration=false)
   &FN_UpdateRibbon(Enable Skeletal Position=true)
   &ScaleColor(
    Scale Alpha=&FloatFromCurve(
     FloatCurve=DataInterface() {propertyValues {Curve {keys [
         {arriveTangent=-16.7667, interpMode=RCIM_Cubic, leaveTangent=-16.7667, tangentMode=RCTM_User, time=-0.0044, value=10}
         {arriveTangent=-7.9091, interpMode=RCIM_Cubic, leaveTangent=-7.9091, tangentMode=RCTM_User, time=0.1708, value=2.2332}
         {arriveTangent=-2.6227, interpMode=RCIM_Cubic, leaveTangent=-2.6227, tangentMode=RCTM_User, time=1.0011, value=0.002}
        ]}}}
     Scale Curve=0.1
    )
    ScaleA=true
    ScaleRGB=true
   ) {module=ScaleColor001}
  ]
  rendererClasses [&NiagaraRibbonRendererProperties]
  renderers [{rendererClass=&NiagaraRibbonRendererProperties}]
 }
 Ship_FE {
  emitterUpdateScript [
   &EmitterState()
   &SpawnBurst_Instantaneous(
    Age:NiagaraFloat=$Emitter.Age
    Spawn Count=&MakeIntFromFloat(Float=&FN_FrontEndFloatModifier(Gameplay=0, Front End=1))
    Use Spawn Probability=false
    UseLoopCountLimit=false
   )
   &InterpolateOverTime(Rate Of Change=50, Target Vector Value:Vector3f=$Engine.Owner.Velocity, Variable Type=::NewEnumerator2)
  ]
  particleSpawnScript [
   &InitializeParticle(
    Color Mode=::NewEnumerator0
    Lifetime=5
    Mesh Scale Mode=::NewEnumerator3
    Position Offset=<0 0 1000>
    UsePositionOffset=true
    Write Material Random=true
   )
   &ShapeLocation(MaskSpawnGroup=false, Override Seed=false, UseOwnerScale=true) {enabled=false}
   &DynamicMaterialParameters(
    Index 0 Param 1=0.15
    Index 0 Param 2=&Multiply_Float(A=$Particles.Lifetime, B=&RandomRangeFloat(Minimum=0.16, Maximum=0.18))
    Index 0 Param 3=0.1
    Index 0 Param 4=0.05
    Module.Param0WriteEnabled=true
    Module.Param1WriteEnabled=true
    Module.Param2WriteEnabled=true
    Module.Param3WriteEnabled=true
   ) {module=DynamicMaterialParameters001}
   &InitialMeshOrientation(Mesh Orientation Mode=::NewEnumerator7, Use Rotation Vector=true)
  ]
  particleUpdateScript [
   &ParticleState(Kill Particles When Lifetime Has Elapsed=false, Let Infinitely Lived Particles Die When Emitter Deactivates=true)
   /SetVariables(Particles.Ambient:Vector3f=&Multiply_VectorByFloat(
     Vector=&MakeVector(
      X=&Waveform(
       "Waveform [1]"=::NewEnumerator12
       "[1] Input Position"=&RandomRangeFloat(Minimum=0, Maximum=1000)
       "[1] Frequency"=1
       "[1] Amplitude Min/Max"=<-200 200>
      )
      Y=&Waveform(
       "Waveform [1]"=::NewEnumerator12
       "[1] Input Position"=&RandomRangeFloat(Minimum=1200, Maximum=1800)
       "[1] Frequency"=1
       "[1] Amplitude Min/Max"=<-100 -500>
      )
      Z=&Waveform(
       "Waveform [1]"=::NewEnumerator12
       "[1] Input Position"=&RandomRangeFloat(Minimum=2200, Maximum=2800)
       "[1] Frequency"=1
       "[1] Amplitude Min/Max"=<-100 50>
      )
     )
     Float=0.2
    )) {isSetParameters=true, module=SetVariables_#5}
   &SkeletalMeshLocation(
    Bone / Socket Index=5
    Bone Sampling Mode=::NewEnumerator4
    MaskSpawnGroup=false
    Orientation Sampling=::NewEnumerator1
    Position Sampling=::NewEnumerator2
    Sampled Position Offset=&RotateVector(
     Vector To Rotate:Vector3f=&Add_Vector(A=$Particles.Ambient, B=<0 75 160>)
     Yaw:NiagaraFloat=&Waveform(
      "Waveform [1]"=::NewEnumerator12
      "[1] Input Position"=&RandomRangeFloat(Minimum=200000, Maximum=100000)
      "[1] Frequency"=1.25
      "[1] Amplitude Min/Max"=<5 -5>
     )
     Pitch:NiagaraFloat=0
     Roll:NiagaraFloat=0
     Rotation Rate:NiagaraFloat=1
    )
    Skeletal Mesh=$User.SkeletalMesh
    Use Sampled Mesh Attribute as Ribbon ID=false
   )
   &ScaleMeshSize(Scale Factor=&VectorFromFloat(Value=&FloatFromCurve(
      FloatCurve=DataInterface() {propertyValues {Curve {keys [{arriveTangent=1.0186, interpMode=RCIM_Cubic, leaveTangent=1.0186, tangentMode=RCTM_User, time=0, value=0}, <at 1 1 RCIM_Cubic>]}}}
      CurveIndex=$Particles.Age
     )))
   &DynamicMaterialParameters(
    Index 0 Param 1=1
    Module.Param0WriteEnabled=true
    Module.Param1WriteEnabled=true
    Module.Param2WriteEnabled=true
    Module.Param3WriteEnabled=true
   ) {enabled=false}
   &PointAttractionForce(
    Attraction Radius=1500
    Attraction Strength=250
    Attractor Position:NiagaraPosition=$Particles.SkeletalMeshLocation.SampledPosition
    Falloff Exponent=0.75
    Kill Within Radius=false
    Module.Use Falloff=true
    UseAttractorOffset=false
   ) {enabled=false}
   &Drag(Drag=2, Use Linear Drag=true, Use Rotational Drag=false)
   &SolveForcesAndVelocity(Clamp Velocity=false, Limit Acceleration=false)
   &UpdateMeshOrientation(
    Bank Rate=30
    "Foward Vector (Mesh Space)":Vector3f=<0 0 1>
    Orientation Method=::NewEnumerator3
    Particle Velocity:Vector3f=$Emitter.InterpolateOverTime.InterpolateOverTime_V3_.Moving Average
    "Right Vector (Mesh Space)":Vector3f=<1 0 0>
    Up Vector:Vector3f=&Lerp_Vector(
     A=&ReturnMeshOrientationAxis(Mesh Orientation Axis=::XAxis, Mesh Orientation Quat=$Particles.SkeletalMeshLocation.SampledRotation)
     B=&Multiply_VectorByFloat(
      Vector=&ReturnMeshOrientationAxis(Mesh Orientation Axis=::YAxis, Mesh Orientation Quat=$Particles.SkeletalMeshLocation.SampledRotation)
      Float=-1
     )
     Alpha=&FloatFromCurve(FloatCurve=DataInterface() {propertyValues {Curve {keys [<at 0 0>,<at 1 1>]}}}, CurveIndex=$User.ParaGlideLeanAlpha)
    )
    Upright Roll Correction Speed=45
    Velocity Orientation Convergence Rate=18
   ) {enabled=false}
   &UpdateMeshOrientation(
    Bank Rate=900000000
    Banking Angle Min / Max:Vector2f=<0.1 -0.1>
    "Foward Vector (Mesh Space)":Vector3f=&Lerp_Vector(A=<0.5 0 0.5>, B=<1 0 0>, Alpha=$User.ParaGlideLeanAlpha)
    Orientation Method=::NewEnumerator3
    Particle Velocity:Vector3f=&Multiply_VectorByFloat(
     Vector=&Add_Vector(
      A=&ReturnMeshOrientationAxis(Mesh Orientation Axis=::YAxis, Mesh Orientation Quat=$Particles.SkeletalMeshLocation.SampledRotation)
      B=&Lerp_Vector(A=<0 0 -2>, B=<0 0 -0.5>, Alpha=$User.ParaGlideLeanAlpha)
     )
     Float=500000
    )
    Rotation Decay Rate In Degrees=900000000
    Up Vector:Vector3f=$Engine.Owner.SystemYAxis
    Upright Roll Correction Speed=900000000
    Velocity Orientation Convergence Rate=900000000
   ) {module=UpdateMeshOrientation001}
   &DynamicMaterialParameters(
    Index 0 Param 1=&RemapRange(Input Value To Remap=$System.Timeline002.NormalizedTime, Input Min=0, Input Max=1, Output Min=0.15, Output Max=1)
    Module.Param0WriteEnabled=true
    Module.Param1WriteEnabled=false
    Module.Param2WriteEnabled=false
    Module.Param3WriteEnabled=false
   ) {module=DynamicMaterialParameters002}
   &InterpolateOverTime(
    Initial Vector Value:NiagaraPosition=$Engine.Owner.Position
    Rate Of Change=500
    Target Vector Value:NiagaraPosition=$Particles.SkeletalMeshLocation.SampledPosition
    Variable Type=::NewEnumerator2
   ) {module=InterpolateOverTime001}
   &InterpolateOverTime(Rate Of Change=5, Target Quat Value:Quat4f=$Particles.MeshOrientation, Variable Type=::NewEnumerator5) {module=InterpolateOverTime002}
   /SetVariables(Particles.Position=$Particles.InterpolateOverTime001.InterpolateOverTime_V3_.Moving Average) {isSetParameters=true}
   /SetVariables(Particles.New Orient:Quat4f=$Particles.InterpolateOverTime002.InterpolateOverTime_Quaternion.Moving Average) {isSetParameters=true, module=SetVariables_#6}
   /SetVariables(Particles.MeshOrientation=$Particles.New Orient) {isSetParameters=true, module=SetVariables_#7}
   /SetVariables(Particles.SpeedConeColor:LinearColor=&MakeLinearColorFromVectorAndFloat("Vector (RGB)"=<1 1 1>, "Float (Alpha)"=&Multiply_Float(A=$User.ParaGlideLeanAlpha, B=0.3))) {isSetParameters=true}
   &self:SphereBlockingVolume_06(Sphere Origin:Vector3f=&SimulationPosition(), Sphere Radius=120) {module=SphereBlockingVolume_06}
  ]
  rendererClasses [&NiagaraMeshRendererProperties]
  renderers [
   {rendererClass=&NiagaraMeshRendererProperties, rendererIndex=0}
   {rendererClass=&NiagaraMeshRendererProperties, rendererIndex=1}
  ]
 }
}
summary {
 systemName=NS_Contrail_CrispRover
 emitters [
  {bEnabled=false, emitterName=Cloud, rendererClasses [&NiagaraMeshRendererProperties]}
  {emitterName=Bird, rendererClasses [&NiagaraMeshRendererProperties, &NiagaraSpriteRendererProperties]}
  {emitterName=Ship_FE, rendererClasses [&NiagaraMeshRendererProperties]}
  {bEnabled=false, emitterName=Cloud_FE, rendererClasses [&NiagaraMeshRendererProperties]}
  {bEnabled=false, emitterName=Cloud_Low, rendererClasses [&NiagaraMeshRendererProperties]}
  {bEnabled=false, emitterName=Cloud_FE_Low, rendererClasses [&NiagaraMeshRendererProperties]}
  {emitterName=Ribbons_Light_Chroma, rendererClasses [&NiagaraRibbonRendererProperties]}
  {emitterName=Hand_Rocks, rendererClasses [&NiagaraMeshRendererProperties]}
  {emitterName=CloudMed, rendererClasses [&NiagaraSpriteRendererProperties]}
  {emitterName=CloudLarge, rendererClasses [&NiagaraSpriteRendererProperties]}
  {emitterName=Cloud_Wispy, rendererClasses [&NiagaraSpriteRendererProperties]}
 ]
 userVariables [
  User.bIsFrontEnd {type=NiagaraBool}
  User.bIsFrontEndPreview {type=NiagaraBool}
  User.FeetRestingPositions {
   type=NiagaraDataInterfaceArrayFloat3
   description=""
   defaultValue {
    struct=NiagaraExt_VariableValue_DataInterface
    value {
     dataInterface=&self:NiagaraDataInterfaceArrayFloat3_0
     dataInterfaceClass=&NiagaraDataInterfaceArrayFloat3
    }
   }
  }
  User.ForwardDot {type=NiagaraFloat, defaultValue=0, description=""}
  User.Landing {type=NiagaraBool}
  User.ParaGlideLeanAlpha {type=NiagaraFloat, defaultValue=0, description=""}
  User.PlayerSpeed {type=NiagaraFloat, defaultValue=0, description=""}
  User.RibbonIDOffset {type=NiagaraInt32}
  User.SkeletalMesh {
   type=NiagaraDataInterfaceSkeletalMesh
   defaultValue {
    struct=NiagaraExt_VariableValue_DataInterface
    value {dataInterface=&self:NiagaraDataInterfaceSkeletalMesh_0}
   }
  }
  User.SpawnCount {type=NiagaraInt32}
 ]
}
